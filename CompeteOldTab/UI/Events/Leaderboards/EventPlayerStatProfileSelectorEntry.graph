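# Widget Blueprint for a player entry in the event leaderboard stat profile selector, built on the native FortEventPlayerStatProfileSelectorEntry.
# It lays out a Select input action prompt, a player flag and a scrolling player name over a hidden skewed-box highlight.
# On hover it plays the Selected animation forward, fading the highlight in and turning the name text black, then reverses it on unhover.
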WidgetBlueprint /CompeteOldTab/UI/Events/Leaderboards/EventPlayerStatProfileSelectorEntry
animations [Selected {
  binding_count=2
  length_seconds=0.2465
  object=&self:Selected
  sequence {bindings [
    PlayerName {
     id=#1
     possessed_class=CommonTextBlock
     tracks [{
       type=MovieSceneColorTrack
       display_name=Color and Opacity
       sections [{
         type=MovieSceneColorSection
         end_frame=4
         start_frame=0
         channels [
          Color.R {type=MovieSceneScriptingFloatChannel, keys [[0,1],[4.9293,0]]}
          Color.G {type=MovieSceneScriptingFloatChannel, keys [[0,1],[4.9293,0]]}
          Color.B {type=MovieSceneScriptingFloatChannel, keys [[0,1],[4.9293,0]]}
          Color.A {type=MovieSceneScriptingFloatChannel}
         ]
        }]
      }]
    }
    Image_SelectedBG {
     id=#2
     possessed_class=Image
     tracks [{
       type=MovieSceneFloatTrack
       display_name=Render Opacity
       sections [{type=MovieSceneFloatSection, end_frame=4, start_frame=0, channels [None {type=MovieSceneScriptingFloatChannel, keys [[0,0],[4.9293,1]]}]}]
      }]
    }
   ]}
 }]
graphs [
 EventGraph {nodes [
   Event On Hovered()
   n1 = Get Selected() -> (Selected:Widget Animation Object Reference)
   > PlayAnimationForward(InAnimation=@n1, PlaybackSpeed=4)
   Event On Unhovered()
   > PlayAnimationReverse(InAnimation=@n1, PlaybackSpeed=3)
  ]}
 Hovered {nodes [Hovered() {is=title K2Node_FunctionEntry}]}
 Unhovered {nodes [Unhovered() {is=title K2Node_FunctionEntry}]}
]
variables [
 Hovered {}
 Unhovered {}
]
widgets {
 info {
  inheritedWidgetCount=3
  parentClass=/Script/FortniteUI.FortEventPlayerStatProfileSelectorEntry
  rootWidgetClass=&Overlay
  widgetCount=9
 }
 widgets [
  Overlay_0 {widgetClassPath=&Overlay, widget=Overlay_0}
  Image_SelectedBG {
   widgetClassPath=&Image
   bIsVariable=true
   parent=Overlay_0
   widget=Image_SelectedBG
   properties {
    renderOpacity=0
    brush {
     bIsDynamicallyLoaded=false
     drawAs=Box
     imageType=NoImage
     mirroring=NoMirror
     resourceName=None
     resourceObject=/Game/UI/Frontend/Showdown/Materials/M_UI_SkewBox_SelectedBG
     tiling=NoTile
     imageSize=<256 64>
     margin=<ltrb 0.4 0.2 0.4 0.2>
     outlineSettings {
      bUseBrushTransparency=false
      roundingType=HalfHeightRadius
      width=0
      color {colorUseRule=UseColor_Specified, specifiedColor=<rgba 0 0 0 0>}
      cornerRadii=<0 0 0 0>
     }
     tintColor {colorUseRule=UseColor_Specified, specifiedColor=<rgba 1 1 1 1>}
     uVRegion {bIsValid=false, max=<0 0>, min=<0 0>}
    }
   }
   slot {class=OverlaySlot, properties {horizontalAlignment=HAlign_Fill, verticalAlignment=VAlign_Fill, padding=<ltrb -10 0 -15 0>}}
  }
  HorizontalBox_0 {
   widgetClassPath=&HorizontalBox
   parent=Overlay_0
   widget=HorizontalBox_0
   properties {visibility=Visible}
   slot {class=OverlaySlot, properties {horizontalAlignment=HAlign_Fill}}
  }
  InputActionWidget {
   widgetClassPath=&CommonActionWidget
   bInherited=true
   bIsVariable=true
   parent=HorizontalBox_0
   widget=InputActionWidget
   properties {
    inputActions [{dataTable=&InputActionDataTable, rowName=Select}]
    renderTransform {angle=0, scale=<1 1>, shear=<0 0>, translation=<-8 0>}
   }
   slot {class=HorizontalBoxSlot, properties {horizontalAlignment=HAlign_Left, verticalAlignment=VAlign_Center}}
  }
  SizeBox_0 {
   widgetClassPath=&SizeBox
   parent=HorizontalBox_0
   widget=SizeBox_0
   properties {bOverride_HeightOverride=true, heightOverride=55, visibility=Visible, widthOverride=450}
   slot {class=HorizontalBoxSlot, properties {horizontalAlignment=HAlign_Center, size {sizeRule=Fill, value=1}}}
  }
  HorizontalBox_1 {
   widgetClassPath=&HorizontalBox
   parent=SizeBox_0
   widget=HorizontalBox_1
   slot {class=SizeBoxSlot, properties {verticalAlignment=VAlign_Center}}
  }
  ScaleBox_0 {
   widgetClassPath=&ScaleBox
   parent=HorizontalBox_1
   widget=ScaleBox_0
   slot {class=HorizontalBoxSlot, properties {horizontalAlignment=HAlign_Center, padding=<ltrb 0 0 10 0>}}
  }
  FlagImage {
   widgetClassPath=/Script/FortniteUI.FortFlagImage
   bInherited=true
   bIsVariable=true
   parent=ScaleBox_0
   widget=FlagImage
   properties {
    materialTextureParamName=Texture
    brush {
     bIsDynamicallyLoaded=false
     drawAs=Image
     imageType=NoImage
     mirroring=NoMirror
     resourceName=None
     resourceObject=/Game/UI/Events/Leaderboards/Materials/M_UI_FlagMaterial_Inst
     tiling=NoTile
     imageSize=<128 128>
     margin=<ltrb 0 0 0 0>
     outlineSettings {
      bUseBrushTransparency=false
      roundingType=HalfHeightRadius
      width=0
      color {colorUseRule=UseColor_Specified, specifiedColor=<rgba 0 0 0 0>}
      cornerRadii=<0 0 0 0>
     }
     tintColor {colorUseRule=UseColor_Specified, specifiedColor=<rgba 1 1 1 1>}
     uVRegion {bIsValid=false, max=<0 0>, min=<0 0>}
    }
   }
   slot {class=ScaleBoxSlot, properties {horizontalAlignment=HAlign_Right}}
  }
  PlayerName {
   widgetClassPath=&CommonTextBlock
   bInherited=true
   bIsVariable=true
   parent=HorizontalBox_1
   widget=PlayerName
   properties {
    minDesiredWidth=440
    scrollStyle=&TextScrollStyle-Base_C
    text="{Player Name}"
    visibility=Visible
    font {
     bForceMonospaced=false
     bMaterialIsStencil=false
     fontMaterial=null
     fontObject=&BurbankSmall
     letterSpacing=0
     monospacedWidth=1
     size=24
     skewAmount=0
     typefaceFontName=Bold
     outlineSettings {
      bApplyOutlineToDropShadows=false
      bMiteredCorners=false
      bSeparateFillAlpha=false
      outlineMaterial=null
      outlineSize=0
      outlineColor=<rgba 0 0 0 1>
     }
    }
    shadowOffset=<0 0>
   }
   slot {class=HorizontalBoxSlot, properties {horizontalAlignment=HAlign_Center, verticalAlignment=VAlign_Center, size {sizeRule=Fill, value=1}}}
  }
 ]
}
