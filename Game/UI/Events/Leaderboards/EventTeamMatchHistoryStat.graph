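# Event leaderboard match-history stat widget that shows a numbered icon beside a score text, built on the native FortEventTeamMatchHistoryStat class.
# PreConstruct passes the Icon Text variable to the numbered icon's stat text.
# Set Hovered and Set Unhovered play the short Selected animation forward or in reverse, fading the score's color from white to black, and forward the same event to the icon.
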WidgetBlueprint /Game/UI/Events/Leaderboards/EventTeamMatchHistoryStat
animations [Selected {
  binding_count=1
  length_seconds=0.25
  object=&self:Selected
  sequence {bindings [Widget_Score {
     id=#1
     possessed_class=CommonTextBlock
     tracks [{
       type=MovieSceneColorTrack
       display_name=Color and Opacity
       sections [{
         type=MovieSceneColorSection
         end_frame=5
         start_frame=0
         channels [
          Color.R {type=MovieSceneScriptingFloatChannel, keys [[0,1],[5,0]]}
          Color.G {type=MovieSceneScriptingFloatChannel, keys [[0,1],[5,0]]}
          Color.B {type=MovieSceneScriptingFloatChannel, keys [[0,1],[5,0]]}
          Color.A {type=MovieSceneScriptingFloatChannel, const [1,2]}
         ]
        }]
      }]
    }]}
 }]
graphs [EventGraph {nodes [
   Set Unhovered Event() {is=title K2Node_CustomEvent}
   n1 = Get Selected() -> (Selected:Widget Animation Object Reference)
   > PlayAnimationReverse(InAnimation=@n1)
   > n2 = Cast To EventTeamMatchHistoryStatNumberedIcon!(execute, Object=Get Widget_Icon() -> (Widget_Icon:Widget Object Reference)) -> (then, CastFailed, AsEvent Team Match History Stat Numbered Icon:Event Team Match History Stat Numbered Icon Object Reference) {is=title K2Node_DynamicCast}
   > Set Unhovered Event(self:Event Team Match History Stat Numbered Icon Object Reference=@n2) {is=title K2Node_CallFunction}
   Set Hovered Event() {is=title K2Node_CustomEvent}
   > PlayAnimationForward(InAnimation=@n1)
   > n3 = Cast To EventTeamMatchHistoryStatNumberedIcon(Object=Get Widget_Icon())
   > Set Hovered Event(self:Event Team Match History Stat Numbered Icon Object Reference=@n3) {is=title K2Node_CallFunction}
   Event PreConstruct()
   > n4 = Cast To EventTeamMatchHistoryStatNumberedIcon(Object=Get Widget_Icon())
   > Set Stat Text!(execute, self:Event Team Match History Stat Numbered Icon Object Reference=@n4, NewText:Text=Get Icon Text() -> (Icon Text:Text)) -> (then) {is=title K2Node_CallFunction}
  ]}]
variables [Icon Text {type=string, default=""}]
widgets {
 info {
  inheritedWidgetCount=2
  parentClass=/Script/FortniteUI.FortEventTeamMatchHistoryStat
  rootWidgetClass=&HorizontalBox
  widgetCount=3
 }
 widgets [
  HorizontalBox_66 {widgetClassPath=&HorizontalBox, widget=HorizontalBox_66, properties {bIsEnabled=false}}
  Widget_Icon {
   widgetClassPath=&/EventTeamMatchHistoryStatNumberedIcon
   bInherited=true
   parent=HorizontalBox_66
   widget=Widget_Icon
   properties {statColor {colorUseRule=UseColor_Specified, specifiedColor=<rgba 0.02 0.04 0.25 1>}}
   slot {class=HorizontalBoxSlot, properties {horizontalAlignment=HAlign_Center, verticalAlignment=VAlign_Center, padding=<ltrb 0 0 5 0>}}
  }
  Widget_Score {
   widgetClassPath=&CommonTextBlock
   bInherited=true
   parent=HorizontalBox_66
   widget=Widget_Score
   properties {
    minDesiredWidth=100
    text="{0}"
    font {
     bForceMonospaced=false
     bMaterialIsStencil=false
     fontMaterial=null
     fontObject=&Burbank
     letterSpacing=0
     monospacedWidth=1
     size=30
     skewAmount=0
     typefaceFontName=Burbank-Black
     outlineSettings {
      bApplyOutlineToDropShadows=false
      bMiteredCorners=false
      bSeparateFillAlpha=false
      outlineMaterial=null
      outlineSize=0
      outlineColor=<rgba 0 0 0 1>
     }
    }
    shadowOffset=<0 0>
   }
   slot {class=HorizontalBoxSlot, properties {horizontalAlignment=HAlign_Center, verticalAlignment=VAlign_Center}}
  }
 ]
}
