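# List row widget for team match history in event leaderboards, showing match number, time alive, eliminations, placement, alternate stats and a Watch replay prompt.
# Hover plays the Selected animation forward or in reverse, swapping the background and darkening text and icons, and macros relay hover state to four stat child widgets.
# Setting the list item zebra-stripes by match-number parity and dims non-score-contributing matches, with no opacity reset in this graph.
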WidgetBlueprint /CompeteOldTab/UI/Events/Leaderboards/EventTeamMatchHistoryEntry
animations [Selected {
  binding_count=10
  length_seconds=0.25
  object=&self:Selected
  sequence {bindings [
    Text_MatchNumber {
     id=#1
     possessed_class=CommonTextBlock
     tracks [{
       type=MovieSceneColorTrack
       display_name=Color and Opacity
       sections [{
         type=MovieSceneColorSection
         end_frame=5
         start_frame=0
         channels [
          Color.R {type=MovieSceneScriptingFloatChannel, keys [[0,1],[4.9293,0],[5,0]]}
          Color.G {type=MovieSceneScriptingFloatChannel, keys [[0,1],[4.9293,0],[5,0]]}
          Color.B {type=MovieSceneScriptingFloatChannel, keys [[0,1],[4.9293,0],[5,0]]}
          Color.A {type=MovieSceneScriptingFloatChannel, const [1,2]}
         ]
        }]
      }]
    }
    Text_Placement {
     id=#2
     possessed_class=CommonTextBlock
     tracks [{
       type=MovieSceneColorTrack
       display_name=Color and Opacity
       sections [{
         type=MovieSceneColorSection
         end_frame=5
         start_frame=0
         channels [
          Color.R {type=MovieSceneScriptingFloatChannel, keys [[0,1],[5,0]]}
          Color.G {type=MovieSceneScriptingFloatChannel, keys [[0,1],[5,0]]}
          Color.B {type=MovieSceneScriptingFloatChannel, keys [[0,1],[5,0]]}
          Color.A {type=MovieSceneScriptingFloatChannel, const [1,2]}
         ]
        }]
      }]
    }
    Text_TimeAlive {
     id=#3
     possessed_class=CommonTextBlock
     tracks [{
       type=MovieSceneColorTrack
       display_name=Color and Opacity
       sections [{
         type=MovieSceneColorSection
         end_frame=5
         start_frame=0
         channels [
          Color.R {type=MovieSceneScriptingFloatChannel, keys [[0,1],[5,0]]}
          Color.G {type=MovieSceneScriptingFloatChannel, keys [[0,1],[5,0]]}
          Color.B {type=MovieSceneScriptingFloatChannel, keys [[0,1],[5,0]]}
          Color.A {type=MovieSceneScriptingFloatChannel, const [1,2]}
         ]
        }]
      }]
    }
    Text_CreativeScore {
     id=#4
     possessed_class=CommonTextBlock
     tracks [{
       type=MovieSceneColorTrack
       display_name=Color and Opacity
       sections [{
         type=MovieSceneColorSection
         end_frame=5
         start_frame=0
         channels [
          Color.R {type=MovieSceneScriptingFloatChannel, keys [[0,1],[5,0]]}
          Color.G {type=MovieSceneScriptingFloatChannel, keys [[0,1],[5,0]]}
          Color.B {type=MovieSceneScriptingFloatChannel, keys [[0,1],[5,0]]}
          Color.A {type=MovieSceneScriptingFloatChannel, const [1,2]}
         ]
        }]
      }]
    }
    Icon_Eliminations {
     id=#5
     possessed_class=Image
     tracks [{
       type=MovieSceneColorTrack
       display_name=Color and Opacity
       sections [{
         type=MovieSceneColorSection
         end_frame=5
         start_frame=0
         channels [
          Color.R {type=MovieSceneScriptingFloatChannel, keys [[0,1],[5,0]]}
          Color.G {type=MovieSceneScriptingFloatChannel, keys [[0,1],[5,0]]}
          Color.B {type=MovieSceneScriptingFloatChannel, keys [[0,1],[5,0]]}
          Color.A {type=MovieSceneScriptingFloatChannel, const [1,2]}
         ]
        }]
      }]
    }
    Icon_CreativeScore {
     id=#6
     possessed_class=Image
     tracks [{
       type=MovieSceneColorTrack
       display_name=Color and Opacity
       sections [{
         type=MovieSceneColorSection
         end_frame=5
         start_frame=0
         channels [
          Color.R {type=MovieSceneScriptingFloatChannel, keys [[0,1],[5,0]]}
          Color.G {type=MovieSceneScriptingFloatChannel, keys [[0,1],[5,0]]}
          Color.B {type=MovieSceneScriptingFloatChannel, keys [[0,1],[5,0]]}
          Color.A {type=MovieSceneScriptingFloatChannel, const [1,2]}
         ]
        }]
      }]
    }
    Text_Eliminations {
     id=#7
     possessed_class=CommonTextBlock
     tracks [{
       type=MovieSceneColorTrack
       display_name=Color and Opacity
       sections [{
         type=MovieSceneColorSection
         end_frame=5
         start_frame=0
         channels [
          Color.R {type=MovieSceneScriptingFloatChannel, keys [[0,1],[5,0]]}
          Color.G {type=MovieSceneScriptingFloatChannel, keys [[0,1],[5,0]]}
          Color.B {type=MovieSceneScriptingFloatChannel, keys [[0,1],[5,0]]}
          Color.A {type=MovieSceneScriptingFloatChannel, const [1,2]}
         ]
        }]
      }]
    }
    Image_SelectedBG {
     id=#8
     possessed_class=Image
     tracks [{
       type=MovieSceneFloatTrack
       display_name=Render Opacity
       sections [{type=MovieSceneFloatSection, end_frame=4, start_frame=0, channels [None {type=MovieSceneScriptingFloatChannel, keys [[0,0],[4.9293,1]]}]}]
      }]
    }
    Icon_Placement {
     id=#9
     possessed_class=Image
     tracks [{
       type=MovieSceneColorTrack
       display_name=Color and Opacity
       sections [{
         type=MovieSceneColorSection
         end_frame=5
         start_frame=0
         channels [
          Color.R {type=MovieSceneScriptingFloatChannel, keys [[0,1],[5,0]]}
          Color.G {type=MovieSceneScriptingFloatChannel, keys [[0,1],[5,0]]}
          Color.B {type=MovieSceneScriptingFloatChannel, keys [[0,1],[5,0]]}
          Color.A {type=MovieSceneScriptingFloatChannel, const [1,2]}
         ]
        }]
      }]
    }
    Text_ReplayInteraction {
     id=#10
     possessed_class=CommonTextBlock
     tracks [{
       type=MovieSceneColorTrack
       display_name=Color and Opacity
       sections [{
         type=MovieSceneColorSection
         end_frame=5
         start_frame=0
         channels [
          Color.R {type=MovieSceneScriptingFloatChannel, keys [[0,1],[5,0]]}
          Color.G {type=MovieSceneScriptingFloatChannel, keys [[0,1],[5,0]]}
          Color.B {type=MovieSceneScriptingFloatChannel, keys [[0,1],[5,0]]}
          Color.A {type=MovieSceneScriptingFloatChannel, const [1,2]}
         ]
        }]
      }]
    }
   ]}
 }]
graphs [
 EventGraph {nodes [
   Event On Hovered()
   n1 = Get Selected() -> (Selected:Widget Animation Object Reference)
   > PlayAnimationForward(InAnimation=@n1)
   > Set Hovered Stat Widget!(execute:Exec, StatWidget:Object Reference=Get Widget_AltStat0() -> (Widget_AltStat0:Fort Event Team Match History Stat Object Reference)) -> (then:Exec) {is=title K2Node_MacroInstance}
   > Set Hovered Stat Widget(StatWidget=Get Widget_AltStat1() -> (Widget_AltStat1:Fort Event Team Match History Stat Object Reference))
   > Set Hovered Stat Widget(StatWidget=Get Widget_AltStat2() -> (Widget_AltStat2:Fort Event Team Match History Stat Object Reference))
   > Set Hovered Stat Widget(StatWidget=Get Widget_AltStat3() -> (Widget_AltStat3:Fort Event Team Match History Stat Object Reference))
   Event On Unhovered()
   > PlayAnimationReverse(InAnimation=@n1)
   > Set Unhovered Stat Widget!(execute:Exec, StatWidget:Object Reference=Get Widget_AltStat0()) -> (then:Exec) {is=title K2Node_MacroInstance}
   > Set Unhovered Stat Widget(StatWidget=Get Widget_AltStat1())
   > Set Unhovered Stat Widget(StatWidget=Get Widget_AltStat2())
   > Set Unhovered Stat Widget(StatWidget=Get Widget_AltStat3())
   n2 = Event OnListItemObjectSet()
   > n3 = Cast To FortEventTeamMatchHistoryEntryData!(execute, Object=@n2.ListItemObject) -> (then, CastFailed, AsFort Event Team Match History Entry Data:Fort Event Team Match History Entry Data Object Reference) {is=title K2Node_DynamicCast}
   n4 = "% (Integer)"(A=Get MatchNumber!(self:Fort Event Team Match History Entry Data Object Reference=@n3) -> (MatchNumber:int=0), B=2) -> (ReturnValue=0)
   > SetBrushColor(
    self=Get Border_Zebra() -> (Border_Zebra:Common Border Object Reference)
    InBrushColor=Select(
     Option 0:Linear Color Structure=(R=1,G=1,B=1,A=0.05)
     Option 1:Linear Color Structure=(R=1,G=1,B=1,A=0.24)
     Index="Equal (Integer)"(A=@n4, B=0) {is=title K2Node_CallFunction}
    ) -> (ReturnValue:Linear Color Structure)
   )
   > n5 = Branch(Condition=Get bIsScoreContributor!(self:Fort Event Team Match History Entry Data Object Reference=@n3) -> (bIsScoreContributor:bool))
   SetRenderOpacity(execute=@n5.else, InOpacity=0.35)
  ]}
 Hovered {nodes [Hovered() {is=title K2Node_FunctionEntry}]}
 Set Hovered Stat Widget {nodes [
   n1 = Inputs!() -> (execute, StatWidget:Object Reference)
   > n2 = Cast To EventTeamMatchHistoryStat!(execute, Object=@n1) -> (then, CastFailed, AsEvent Team Match History Stat:Event Team Match History Stat Object Reference) {is=title K2Node_DynamicCast}
   > n3 = Set Hovered Event(self:Event Team Match History Stat Object Reference=@n2) {is=title K2Node_CallFunction}
   Outputs!(then:Exec=[@n3.then, @n2.CastFailed])
  ]}
 Set Unhovered Stat Widget {nodes [
   n1 = Inputs()
   > n2 = Cast To EventTeamMatchHistoryStat(Object=@n1)
   > n3 = Set Unhovered Event(self:Event Team Match History Stat Object Reference=@n2) {is=title K2Node_CallFunction}
   Outputs(then=[@n3.then, @n2.CastFailed])
  ]}
 Unhovered {nodes [Unhovered() {is=title K2Node_FunctionEntry}]}
]
variables [
 Hovered {}
 Unhovered {}
]
widgets {
 info {inheritedWidgetCount=15, parentClass=/Script/FortniteUI.FortEventTeamMatchHistoryEntry, rootWidgetClass=&Overlay, widgetCount=29}
 widgets [
  Overlay_0 {widgetClassPath=&Overlay, widget=Overlay_0}
  Image_SelectedBG {
   widgetClassPath=&Image
   parent=Overlay_0
   widget=Image_SelectedBG
   properties {
    renderOpacity=0
    brush {
     bIsDynamicallyLoaded=false
     drawAs=Box
     imageType=NoImage
     mirroring=NoMirror
     resourceName=None
     resourceObject=/Game/UI/Frontend/Showdown/Materials/M_UI_SkewBox_SelectedBG
     tiling=NoTile
     imageSize=<256 64>
     margin=<ltrb 0.4 0.2 0.4 0.2>
     outlineSettings {
      bUseBrushTransparency=false
      roundingType=HalfHeightRadius
      width=0
      color {colorUseRule=UseColor_Specified, specifiedColor=<rgba 0 0 0 0>}
      cornerRadii=<0 0 0 0>
     }
     tintColor {colorUseRule=UseColor_Specified, specifiedColor=<rgba 1 1 1 1>}
     uVRegion {bIsValid=false, max=<0 0>, min=<0 0>}
    }
   }
   slot {class=OverlaySlot, properties {horizontalAlignment=HAlign_Fill, verticalAlignment=VAlign_Fill, padding=<ltrb -10 0 -15 0>}}
  }
  Border_Zebra {
   widgetClassPath=&CommonBorder
   bIsVariable=true
   parent=Overlay_0
   widget=Border_Zebra
   properties {
    background {
     bIsDynamicallyLoaded=false
     drawAs=Image
     imageType=NoImage
     mirroring=NoMirror
     resourceName=None
     resourceObject=null
     tiling=NoTile
     imageSize=<32 32>
     margin=<ltrb 0 0 0 0>
     outlineSettings {
      bUseBrushTransparency=false
      roundingType=HalfHeightRadius
      width=0
      color {colorUseRule=UseColor_Specified, specifiedColor=<rgba 0 0 0 0>}
      cornerRadii=<0 0 0 0>
     }
     tintColor {colorUseRule=UseColor_Specified, specifiedColor=<rgba 1 1 1 1>}
     uVRegion {bIsValid=false, max=<0 0>, min=<0 0>}
    }
    brushColor=<rgba 0.0097 0 0.9047 0.0588>
    padding=<ltrb 0 0 0 0>
   }
   slot {class=OverlaySlot, properties {horizontalAlignment=HAlign_Fill, verticalAlignment=VAlign_Fill}}
  }
  HorizontalBox_0 {
   widgetClassPath=&HorizontalBox
   parent=Border_Zebra
   widget=HorizontalBox_0
   properties {visibility=Visible}
   slot {class=BorderSlot, properties {padding=<ltrb 0 0 0 0>}}
  }
  SizeBox_0 {
   widgetClassPath=&SizeBox
   parent=HorizontalBox_0
   widget=SizeBox_0
   properties {bOverride_HeightOverride=true, heightOverride=55, visibility=Visible, widthOverride=450}
   slot {class=HorizontalBoxSlot, properties {horizontalAlignment=HAlign_Center, size {sizeRule=Fill, value=1}}}
  }
  HorizontalBox_1 {
   widgetClassPath=&HorizontalBox
   parent=SizeBox_0
   widget=HorizontalBox_1
   slot {class=SizeBoxSlot, properties {verticalAlignment=VAlign_Center}}
  }
  Overlay_OnlineStatus {
   widgetClassPath=&Overlay
   bInherited=true
   parent=HorizontalBox_1
   widget=Overlay_OnlineStatus
   properties {visibility=Collapsed}
   slot {class=HorizontalBoxSlot, properties {horizontalAlignment=HAlign_Center, verticalAlignment=VAlign_Center}}
  }
  Image_ping {
   widgetClassPath=&Image
   parent=Overlay_OnlineStatus
   widget=Image_ping
   properties {
    visibility=SelfHitTestInvisible
    brush {
     bIsDynamicallyLoaded=false
     drawAs=Image
     imageType=NoImage
     mirroring=NoMirror
     resourceName=None
     resourceObject=/Game/UI/Events/Leaderboards/Materials/M_UI_LiveIcon_Inst
     tiling=NoTile
     imageSize=<60 60>
     margin=<ltrb 0 0 0 0>
     outlineSettings {
      bUseBrushTransparency=false
      roundingType=HalfHeightRadius
      width=0
      color {colorUseRule=UseColor_Specified, specifiedColor=<rgba 0 0 0 0>}
      cornerRadii=<0 0 0 0>
     }
     tintColor {colorUseRule=UseColor_Specified, specifiedColor=<rgba 1 1 1 1>}
     uVRegion {bIsValid=false, max=<0 0>, min=<0 0>}
    }
   }
   slot {class=OverlaySlot, properties {padding=<ltrb 0 -10 0 -10>}}
  }
  Image_OnlineStatus {
   widgetClassPath=&Image
   parent=Overlay_OnlineStatus
   widget=Image_OnlineStatus
   properties {
    visibility=SelfHitTestInvisible
    brush {
     bIsDynamicallyLoaded=false
     drawAs=Image
     imageType=NoImage
     mirroring=NoMirror
     resourceName=None
     resourceObject=/Game/UI/Events/Leaderboards/Materials/M_UI_LiveIconCenter_Inst
     tiling=NoTile
     imageSize=<30 30>
     margin=<ltrb 0 0 0 0>
     outlineSettings {
      bUseBrushTransparency=false
      roundingType=HalfHeightRadius
      width=0
      color {colorUseRule=UseColor_Specified, specifiedColor=<rgba 0 0 0 0>}
      cornerRadii=<0 0 0 0>
     }
     tintColor {colorUseRule=UseColor_Specified, specifiedColor=<rgba 1 1 1 1>}
     uVRegion {bIsValid=false, max=<0 0>, min=<0 0>}
    }
    colorAndOpacity=<rgba 1 0 0 1>
   }
   slot {class=OverlaySlot, properties {horizontalAlignment=HAlign_Center, verticalAlignment=VAlign_Center}}
  }
  ScaleBox_0 {
   widgetClassPath=&ScaleBox
   parent=HorizontalBox_1
   widget=ScaleBox_0
   properties {stretchDirection=DownOnly}
   slot {class=HorizontalBoxSlot, properties {padding=<ltrb 0 0 10 0>}}
  }
  Text_MatchNumber {
   widgetClassPath=&CommonTextBlock
   bInherited=true
   parent=ScaleBox_0
   widget=Text_MatchNumber
   properties {
    minDesiredWidth=75
    text="[ph] ##1"
    font {
     bForceMonospaced=false
     bMaterialIsStencil=false
     fontMaterial=null
     fontObject=&Burbank
     letterSpacing=0
     monospacedWidth=1
     size=30
     skewAmount=0
     typefaceFontName=Burbank-Black
     outlineSettings {
      bApplyOutlineToDropShadows=false
      bMiteredCorners=false
      bSeparateFillAlpha=false
      outlineMaterial=null
      outlineSize=0
      outlineColor=<rgba 0 0 0 1>
     }
    }
    shadowOffset=<0 0>
   }
   slot {class=ScaleBoxSlot, properties {horizontalAlignment=HAlign_Fill, verticalAlignment=VAlign_Fill}}
  }
  ScaleBox_4 {
   widgetClassPath=&ScaleBox
   parent=HorizontalBox_1
   widget=ScaleBox_4
   properties {stretchDirection=DownOnly}
   slot {class=HorizontalBoxSlot, properties {padding=<ltrb 0 0 10 0>}}
  }
  Text_TimeAlive {
   widgetClassPath=&CommonTextBlock
   bInherited=true
   parent=ScaleBox_4
   widget=Text_TimeAlive
   properties {
    minDesiredWidth=300
    text="[ph] 2 minutes"
    font {
     bForceMonospaced=false
     bMaterialIsStencil=false
     fontMaterial=null
     fontObject=&Burbank
     letterSpacing=0
     monospacedWidth=1
     size=30
     skewAmount=0
     typefaceFontName=Burbank-Black
     outlineSettings {
      bApplyOutlineToDropShadows=false
      bMiteredCorners=false
      bSeparateFillAlpha=false
      outlineMaterial=null
      outlineSize=0
      outlineColor=<rgba 0 0 0 1>
     }
    }
    shadowOffset=<0 0>
   }
   slot {class=ScaleBoxSlot, properties {horizontalAlignment=HAlign_Fill}}
  }
  Icon_Eliminations {
   widgetClassPath=&Image
   bInherited=true
   parent=HorizontalBox_1
   widget=Icon_Eliminations
   properties {
    visibility=SelfHitTestInvisible
    brush {
     bIsDynamicallyLoaded=false
     drawAs=Image
     imageType=NoImage
     mirroring=NoMirror
     resourceName=None
     resourceObject=/Game/UI/Foundation/Textures/Icons/Replay/T-Icon-Replay-Filter-Kill-L
     tiling=NoTile
     imageSize=<32 32>
     margin=<ltrb 0 0 0 0>
     outlineSettings {
      bUseBrushTransparency=false
      roundingType=HalfHeightRadius
      width=0
      color {colorUseRule=UseColor_Specified, specifiedColor=<rgba 0 0 0 0>}
      cornerRadii=<0 0 0 0>
     }
     tintColor {colorUseRule=UseColor_Specified, specifiedColor=<rgba 1 1 1 1>}
     uVRegion {bIsValid=false, max=<0 0>, min=<0 0>}
    }
   }
   slot {class=HorizontalBoxSlot, properties {horizontalAlignment=HAlign_Center, verticalAlignment=VAlign_Center, padding=<ltrb 5 0 5 0>}}
  }
  Text_Eliminations {
   widgetClassPath=&CommonTextBlock
   bInherited=true
   parent=HorizontalBox_1
   widget=Text_Eliminations
   properties {
    minDesiredWidth=100
    text="{0}"
    font {
     bForceMonospaced=false
     bMaterialIsStencil=false
     fontMaterial=null
     fontObject=&Burbank
     letterSpacing=0
     monospacedWidth=1
     size=30
     skewAmount=0
     typefaceFontName=Burbank-Black
     outlineSettings {
      bApplyOutlineToDropShadows=false
      bMiteredCorners=false
      bSeparateFillAlpha=false
      outlineMaterial=null
      outlineSize=0
      outlineColor=<rgba 0 0 0 1>
     }
    }
    shadowOffset=<0 0>
   }
   slot {class=HorizontalBoxSlot, properties {horizontalAlignment=HAlign_Center, verticalAlignment=VAlign_Center, padding=<ltrb 0 0 5 0>}}
  }
  Widget_AltStat0 {
   widgetClassPath=/Game/UI/Events/Leaderboards/EventTeamMatchHistoryStat
   bInherited=true
   parent=HorizontalBox_1
   widget=Widget_AltStat0
   properties {icon Text=1}
   slot {class=HorizontalBoxSlot, properties {padding=<ltrb 5 0 5 0>}}
  }
  Widget_AltStat1 {
   widgetClassPath=&EventTeamMatchHistoryStat
   bInherited=true
   parent=HorizontalBox_1
   widget=Widget_AltStat1
   properties {icon Text=2}
   slot {class=HorizontalBoxSlot, properties {padding=<ltrb 5 0 5 0>}}
  }
  Widget_AltStat2 {
   widgetClassPath=&EventTeamMatchHistoryStat
   bInherited=true
   parent=HorizontalBox_1
   widget=Widget_AltStat2
   properties {icon Text=3}
   slot {class=HorizontalBoxSlot, properties {padding=<ltrb 5 0 5 0>}}
  }
  Widget_AltStat3 {
   widgetClassPath=&EventTeamMatchHistoryStat
   bInherited=true
   parent=HorizontalBox_1
   widget=Widget_AltStat3
   properties {icon Text=4}
   slot {class=HorizontalBoxSlot, properties {padding=<ltrb 5 0 5 0>}}
  }
  Icon_Placement {
   widgetClassPath=&Image
   bInherited=true
   parent=HorizontalBox_1
   widget=Icon_Placement
   properties {
    visibility=SelfHitTestInvisible
    brush {
     bIsDynamicallyLoaded=false
     drawAs=Image
     imageType=NoImage
     mirroring=NoMirror
     resourceName=None
     resourceObject=/Game/UI/Foundation/Textures/Icons/Quest/T-Icon-Trophy-64
     tiling=NoTile
     imageSize=<32 32>
     margin=<ltrb 0 0 0 0>
     outlineSettings {
      bUseBrushTransparency=false
      roundingType=HalfHeightRadius
      width=0
      color {colorUseRule=UseColor_Specified, specifiedColor=<rgba 0 0 0 0>}
      cornerRadii=<0 0 0 0>
     }
     tintColor {colorUseRule=UseColor_Specified, specifiedColor=<rgba 1 1 1 1>}
     uVRegion {bIsValid=false, max=<0 0>, min=<0 0>}
    }
   }
   slot {class=HorizontalBoxSlot, properties {horizontalAlignment=HAlign_Center, verticalAlignment=VAlign_Center, padding=<ltrb 5 0 5 0>}}
  }
  Text_Placement {
   widgetClassPath=&CommonTextBlock
   bInherited=true
   parent=HorizontalBox_1
   widget=Text_Placement
   properties {
    minDesiredWidth=100
    text="{0}"
    font {
     bForceMonospaced=false
     bMaterialIsStencil=false
     fontMaterial=null
     fontObject=&Burbank
     letterSpacing=0
     monospacedWidth=1
     size=30
     skewAmount=0
     typefaceFontName=Burbank-Black
     outlineSettings {
      bApplyOutlineToDropShadows=false
      bMiteredCorners=false
      bSeparateFillAlpha=false
      outlineMaterial=null
      outlineSize=0
      outlineColor=<rgba 0 0 0 1>
     }
    }
    shadowOffset=<0 0>
   }
   slot {class=HorizontalBoxSlot, properties {horizontalAlignment=HAlign_Center, verticalAlignment=VAlign_Center}}
  }
  Icon_CreativeScore {
   widgetClassPath=&Image
   bInherited=true
   parent=HorizontalBox_1
   widget=Icon_CreativeScore
   properties {
    visibility=SelfHitTestInvisible
    brush {
     bIsDynamicallyLoaded=false
     drawAs=Image
     imageType=NoImage
     mirroring=NoMirror
     resourceName=None
     resourceObject=/Game/UI/Foundation/Textures/Icons/Athena/T_T_Icon_Athena_AltStat
     tiling=NoTile
     imageSize=<32 32>
     margin=<ltrb 0 0 0 0>
     outlineSettings {
      bUseBrushTransparency=false
      roundingType=HalfHeightRadius
      width=0
      color {colorUseRule=UseColor_Specified, specifiedColor=<rgba 0 0 0 0>}
      cornerRadii=<0 0 0 0>
     }
     tintColor {colorUseRule=UseColor_Specified, specifiedColor=<rgba 1 1 1 1>}
     uVRegion {bIsValid=false, max=<0 0>, min=<0 0>}
    }
   }
   slot {class=HorizontalBoxSlot, properties {horizontalAlignment=HAlign_Center, verticalAlignment=VAlign_Center, padding=<ltrb 5 0 5 0>}}
  }
  Text_CreativeScore {
   widgetClassPath=&CommonTextBlock
   bInherited=true
   parent=HorizontalBox_1
   widget=Text_CreativeScore
   properties {
    minDesiredWidth=100
    text="{0}"
    font {
     bForceMonospaced=false
     bMaterialIsStencil=false
     fontMaterial=null
     fontObject=&Burbank
     letterSpacing=0
     monospacedWidth=1
     size=30
     skewAmount=0
     typefaceFontName=Burbank-Black
     outlineSettings {
      bApplyOutlineToDropShadows=false
      bMiteredCorners=false
      bSeparateFillAlpha=false
      outlineMaterial=null
      outlineSize=0
      outlineColor=<rgba 0 0 0 1>
     }
    }
    shadowOffset=<0 0>
   }
   slot {class=HorizontalBoxSlot, properties {horizontalAlignment=HAlign_Center, verticalAlignment=VAlign_Center}}
  }
  Border_InteractionContent {
   widgetClassPath=&CommonBorder
   bInherited=true
   parent=HorizontalBox_0
   widget=Border_InteractionContent
   slot {class=HorizontalBoxSlot, properties {verticalAlignment=VAlign_Center, padding=<ltrb 10 0 10 0>}}
  }
  HorizontalBox_2 {widgetClassPath=&HorizontalBox, parent=Border_InteractionContent, widget=HorizontalBox_2, slot {class=BorderSlot}}
  SizeBox_1 {
   widgetClassPath=&SizeBox
   parent=HorizontalBox_2
   widget=SizeBox_1
   properties {bOverride_HeightOverride=true, bOverride_WidthOverride=true, heightOverride=32, widthOverride=32}
   slot {class=HorizontalBoxSlot, properties {verticalAlignment=VAlign_Center, padding=<ltrb 0 0 10 0>}}
  }
  InputActionWidget {
   widgetClassPath=&CommonActionWidget
   bInherited=true
   parent=SizeBox_1
   widget=InputActionWidget
   properties {visibility=SelfHitTestInvisible, inputActions [{dataTable=&InputActionDataTable, rowName=Select}]}
   slot {class=SizeBoxSlot}
  }
  ScaleBox_3 {
   widgetClassPath=&ScaleBox
   parent=HorizontalBox_2
   widget=ScaleBox_3
   properties {stretchDirection=DownOnly}
   slot {class=HorizontalBoxSlot, properties {horizontalAlignment=HAlign_Left, padding=<ltrb 0 0 10 0>, size {sizeRule=Fill, value=1}}}
  }
  Text_ReplayInteraction {
   widgetClassPath=&CommonTextBlock
   parent=ScaleBox_3
   widget=Text_ReplayInteraction
   properties {
    scrollStyle=&TextScrollStyle-Base_C
    text=Watch
    font {
     bForceMonospaced=false
     bMaterialIsStencil=false
     fontMaterial=null
     fontObject=&BurbankSmall
     letterSpacing=0
     monospacedWidth=1
     size=24
     skewAmount=0
     typefaceFontName=Bold
     outlineSettings {
      bApplyOutlineToDropShadows=false
      bMiteredCorners=false
      bSeparateFillAlpha=false
      outlineMaterial=null
      outlineSize=0
      outlineColor=<rgba 0 0 0 1>
     }
    }
    shadowOffset=<0 0>
   }
   slot {class=ScaleBoxSlot}
  }
 ]
}
