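# Widget Blueprint on FortShowdownDetailView that shows a one-line tournament title and round label, apparently for the event match history leaderboard UI.
# On Refresh Data, it breaks TournamentDisplayInfo and formats both title lines with GetWindowRoundDisplayString into localized "Title1 Title2 - Round" text.
# The result goes to Text_TournamentInfo, an uppercase root text block; the inherited poster, timer and debug widgets are unbound.
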WidgetBlueprint /Game/UI/Events/Leaderboards/EventMatchHistoryTournamentInfo
graphs [EventGraph {nodes [
   Event Refresh Data() {is=title K2Node_Event}
   n1 = Break Fort Tournament Display Info!(FortTournamentDisplayInfo:"Fort Tournament Display Info Structure (by ref)"=Get TournamentDisplayInfo() -> (TournamentDisplayInfo:Fort Tournament Display Info Structure)) -> (
    TitleLine1:Text
    TitleLine2:Text
    ScheduleInfo:Text
    FlavorDescription:Text
    DetailsDescription:Text
    ShortFormatTitle:Text
    LongFormatTitle:Text
    BackgroundTitle:Text
    PinScoreRequirement:int
    PinEarnedText:Text
    BaseColor:Linear Color Structure
    PrimaryColor:Linear Color Structure
    SecondaryColor:Linear Color Structure
    HighlightColor:Linear Color Structure
    TitleColor:Linear Color Structure
    ShadowColor:Linear Color Structure
    BackgroundLeftColor:Linear Color Structure
    BackgroundRightColor:Linear Color Structure
    BackgroundTextColor:Linear Color Structure
    PosterFadeColor:Linear Color Structure
    AlertText:Text
    SeriesPointLeaderboardName:Text
    AlertType:EFortTournamentAlertType Enum
    RoundDisplayNames:Array of Texts
   ) {is=title K2Node_BreakStruct}
   > "SetText (Text)"(
    self=Get Text_TournamentInfo() -> (Text_TournamentInfo:Common Text Object Reference)
    InText=Format Text!(
     Format="LOCTEXT(\"{TitlePart1} {TitlePart2} - {RoundInfo}\")"
     TitlePart1:Text=@n1
     TitlePart2:Text=@n1.TitleLine2
     RoundInfo:Text=GetWindowRoundDisplayString!(self:Fort Showdown Detail View Object Reference) -> (ReturnValue:Text)
    ) -> (Result)
   )
   Event PreConstruct()
   Event Construct()
   Event Tick()
  ]}]
variables []
widgets {
 info {
  inheritedWidgetCount=6
  parentClass=/Script/FortniteUI.FortShowdownDetailView
  rootWidgetClass=&CommonTextBlock
  widgetCount=7
 }
 widgets [
  Text_TournamentInfo {
   widgetClassPath=&CommonTextBlock
   bIsVariable=true
   widget=Text_TournamentInfo
   properties {
    scrollStyle=&TextScrollStyle-Base_C
    text="[ph] {Tournament Title} - {Round Info}"
    textTransformPolicy=ToUpper
    font {
     bForceMonospaced=false
     bMaterialIsStencil=false
     fontMaterial=null
     fontObject=&BurbankSmall
     letterSpacing=0
     monospacedWidth=1
     size=22
     skewAmount=0
     typefaceFontName=Bold
     outlineSettings {
      bApplyOutlineToDropShadows=false
      bMiteredCorners=false
      bSeparateFillAlpha=false
      outlineMaterial=null
      outlineSize=0
      outlineColor=<rgba 0 0 0 1>
     }
    }
    shadowOffset=<0 0>
   }
  }
  Text_TimeRemaining {widgetClassPath=&CommonTextBlock, bInherited=true, widget=null}
  Text_DebugId {widgetClassPath=&TextBlock, bInherited=true, widget=null, properties=~}
  Image_PosterFront {widgetClassPath=/Script/EpicCMSUIFramework.EpicCMSImage, bInherited=true, widget=null}
  Image_PosterBack {widgetClassPath=&EpicCMSImage, bInherited=true, widget=null}
  Image_LoadingScreen {widgetClassPath=&EpicCMSImage, bInherited=true, widget=null}
  Image_PlaylistTile {widgetClassPath=&EpicCMSImage, bInherited=true, widget=null}
 ]
}
