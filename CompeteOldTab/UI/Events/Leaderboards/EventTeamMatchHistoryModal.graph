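# Modal widget on FortEventTeamMatchHistoryModal that shows a team's match history for a tournament event, with team names, a match list and four disabled stat widgets.
# OnSetupMatchHistory passes tournament and event window IDs to the tournament info widget with a forced refresh.
# Set Up Match History Style tints the title, its shadow and the background from tournament display colors.
# Activation plays a scale, fade and blur intro; Close pops the modal layer.
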WidgetBlueprint /CompeteOldTab/UI/Events/Leaderboards/EventTeamMatchHistoryModal
animations [Intro {
  binding_count=2
  length_seconds=0.2218
  object=&self:Intro
  sequence {bindings [
    SafeZone_0 {
     id=#1
     possessed_class=SafeZone
     tracks [
      {
       type=MovieScene2DTransformTrack
       sections [{
         type=MovieScene2DTransformSection
         channels [
          Translation.X {type=MovieSceneScriptingFloatChannel}
          Translation.Y {type=MovieSceneScriptingFloatChannel}
          Angle {type=MovieSceneScriptingFloatChannel}
          Scale.X {type=MovieSceneScriptingFloatChannel, keys [[0,0.8],[3.76,1.0103],[4.43,1]]}
          Scale.Y {type=MovieSceneScriptingFloatChannel, keys [[0,0.8],[3.76,1.0103],[4.43,1]]}
          Shear.X {type=MovieSceneScriptingFloatChannel}
          Shear.Y {type=MovieSceneScriptingFloatChannel}
         ]
        }]
      }
      {
       type=MovieSceneFloatTrack
       display_name=Render Opacity
       sections [{type=MovieSceneFloatSection, channels [None {type=MovieSceneScriptingFloatChannel, keys [[0.917,0],[4.435,1]]}]}]
      }
     ]
    }
    BackgroundBlur_1 {
     id=#2
     possessed_class=BackgroundBlur
     tracks [{
       type=MovieSceneFloatTrack
       display_name=Blur Strength
       sections [{type=MovieSceneFloatSection, channels [None {type=MovieSceneScriptingFloatChannel, keys [[0,0],[4.435,25]]}]}]
      }]
    }
   ]}
 }]
graphs [
 EventGraph {nodes [
   "On Button Clicked (CloseMainPanel)"!() -> (OutputDelegate, then, Button:Common Button Legacy Object Reference) {is=title K2Node_ComponentBoundEvent}
   > PopActivatablePanelInModalLayer!(execute, Panel:Common Activatable Panel Legacy Object Reference=Self-Reference()) -> (then)
   n1 = Event OnSetupMatchHistory!() -> (OutputDelegate, then, TournamentId:String, EventWindowId:String) {is=title K2Node_Event}
   > SetTournament!(
    execute
    self:Fort Showdown Detail View Object Reference=Get EventMatchHistoryTournamentInfo() -> (EventMatchHistoryTournamentInfo:Event Match History Tournament Info Object Reference)
    TournamentId:String=@n1.TournamentId
    EventId:String=@n1.EventWindowId
    bForceRefreshUI:bool=true
   ) -> (then)
   Event On Activated()
   > PlayAnimationForward(InAnimation=Get Intro())
  ]}
 Set Up Match History Style {nodes [
   n1 = Set Up Match History Style!() -> (then, Tournament Display Info:Fort Tournament Display Info Structure) {is=title K2Node_FunctionEntry}
   n2 = Break Fort Tournament Display Info!(FortTournamentDisplayInfo:"Fort Tournament Display Info Structure (by ref)"=@n1) -> (
    TitleLine1:Text
    TitleLine2:Text
    ScheduleInfo:Text
    FlavorDescription:Text
    DetailsDescription:Text
    ShortFormatTitle:Text
    LongFormatTitle:Text
    BackgroundTitle:Text
    PinScoreRequirement:int
    PinEarnedText:Text
    BaseColor:Linear Color Structure
    PrimaryColor:Linear Color Structure
    SecondaryColor:Linear Color Structure
    HighlightColor:Linear Color Structure
    TitleColor:Linear Color Structure
    ShadowColor:Linear Color Structure
    BackgroundLeftColor:Linear Color Structure
    BackgroundRightColor:Linear Color Structure
    BackgroundTextColor:Linear Color Structure
    PosterFadeColor:Linear Color Structure
    AlertText:Text
    SeriesPointLeaderboardName:Text
    AlertType:EFortTournamentAlertType Enum
    RoundDisplayNames:Array of Texts
   ) {is=title K2Node_BreakStruct}
   n3 = Get Text_WindowTittle() -> (Text_WindowTittle:Common Text Object Reference)
   > SetColorAndOpacity(self=@n3, InColorAndOpacity=Make Slate Color(SpecifiedColor=@n2.BaseColor))
   > SetColorAndOpacity(
    self:Image Object Reference=Get Image_Background() -> (Image_Background:Image Object Reference)
    InColorAndOpacity:Linear Color Structure=@n2.PrimaryColor
   )
   > SetShadowColorAndOpacity!(execute, self:Text Object Reference=@n3, InShadowColorAndOpacity:Linear Color Structure=@n2.ShadowColor) -> (then)
  ]}
]
variables []
widgets {
 info {
  inheritedWidgetCount=6
  parentClass=/Script/FortniteUI.FortEventTeamMatchHistoryModal
  rootWidgetClass=&CommonBorder
  widgetCount=40
 }
 widgets [
  TouchToCloseZone {widgetClassPath=&CommonBorder, widget=TouchToCloseZone, properties {padding=<ltrb 0 0 0 0>}}
  BackgroundBlur_1 {
   widgetClassPath=&BackgroundBlur
   parent=TouchToCloseZone
   widget=BackgroundBlur_1
   properties {
    blurStrength=25
    visibility=Visible
    lowQualityFallbackBrush {
     bIsDynamicallyLoaded=false
     drawAs=Image
     imageType=NoImage
     mirroring=NoMirror
     resourceName=None
     resourceObject=null
     tiling=NoTile
     imageSize=<0 0>
     margin=<ltrb 0 0 0 0>
     outlineSettings {
      bUseBrushTransparency=false
      roundingType=HalfHeightRadius
      width=0
      color {colorUseRule=UseColor_Specified, specifiedColor=<rgba 0 0 0 0>}
      cornerRadii=<0 0 0 0>
     }
     tintColor {colorUseRule=UseColor_Specified, specifiedColor=<rgba 0 0.0006 0.0052 0.749>}
     uVRegion {bIsValid=false, max=<0 0>, min=<0 0>}
    }
   }
   slot {class=BorderSlot, properties {padding=<ltrb 0 0 0 0>}}
  }
  SafeZone_0 {
   widgetClassPath=&SafeZone
   bIsVariable=true
   parent=BackgroundBlur_1
   widget=SafeZone_0
   slot {class=BackgroundBlurSlot, properties {padding=<ltrb 0 0 0 0>}}
  }
  Overlay_0 {widgetClassPath=&Overlay, parent=SafeZone_0, widget=Overlay_0, slot {class=SafeZoneSlot}}
  Switcher_Content {
   widgetClassPath=&WidgetSwitcher
   bIsVariable=true
   parent=Overlay_0
   widget=Switcher_Content
   slot {class=OverlaySlot, properties {horizontalAlignment=HAlign_Center, verticalAlignment=VAlign_Center}}
  }
  SizeBox_0 {
   widgetClassPath=&SizeBox
   parent=Switcher_Content
   widget=SizeBox_0
   properties {bOverride_MinDesiredWidth=true, maxDesiredHeight=900, maxDesiredWidth=1024, minDesiredHeight=400, minDesiredWidth=1024, widthOverride=1024}
   slot {class=WidgetSwitcherSlot}
  }
  Overlay_1 {widgetClassPath=&Overlay, parent=SizeBox_0, widget=Overlay_1, slot {class=SizeBoxSlot}}
  Image_dropshadow {
   widgetClassPath=&Image
   parent=Overlay_1
   widget=Image_dropshadow
   properties {
    renderOpacity=0.44
    colorAndOpacity=<rgba 0.0052 0.0052 0.0056 0.44>
    renderTransform {angle=0, scale=<1 1>, shear=<0 0>, translation=<20 28>}
   }
   slot {class=OverlaySlot, properties {horizontalAlignment=HAlign_Fill, verticalAlignment=VAlign_Fill}}
  }
  Overlay_232 {widgetClassPath=&Overlay, parent=Overlay_1, widget=Overlay_232, slot {class=OverlaySlot}}
  Image_Background {
   widgetClassPath=&Image
   bIsVariable=true
   parent=Overlay_232
   widget=Image_Background
   properties {brush {
     bIsDynamicallyLoaded=false
     drawAs=Image
     imageType=NoImage
     mirroring=NoMirror
     resourceName=None
     resourceObject=null
     tiling=NoTile
     imageSize=<0 32>
     margin=<ltrb 0 0 0 0>
     outlineSettings {
      bUseBrushTransparency=false
      roundingType=HalfHeightRadius
      width=0
      color {colorUseRule=UseColor_Specified, specifiedColor=<rgba 0 0 0 0>}
      cornerRadii=<0 0 0 0>
     }
     tintColor {colorUseRule=UseColor_Specified, specifiedColor=<rgba 0.0009 0.0497 0.3185 1>}
     uVRegion {bIsValid=false, max=<0 0>, min=<0 0>}
    }}
   slot {class=OverlaySlot, properties {horizontalAlignment=HAlign_Fill, verticalAlignment=VAlign_Fill}}
  }
  CommonBorder_0 {
   widgetClassPath=&CommonBorder
   bIsVariable=true
   parent=Overlay_232
   widget=CommonBorder_0
   properties {padding=<ltrb 0 0 0 0>, renderTransform {angle=0, scale=<1 1>, shear=<0 0>, translation=<1 1>}}
   slot {class=OverlaySlot, properties {horizontalAlignment=HAlign_Fill, verticalAlignment=VAlign_Fill}}
  }
  VerticalBox_0 {
   widgetClassPath=&VerticalBox
   parent=CommonBorder_0
   widget=VerticalBox_0
   slot {class=BorderSlot, properties {padding=<ltrb 0 0 0 0>}}
  }
  SizeBox_2 {
   widgetClassPath=&SizeBox
   parent=VerticalBox_0
   widget=SizeBox_2
   properties {bOverride_MaxDesiredWidth=true, maxDesiredWidth=800}
   slot {class=VerticalBoxSlot, properties {padding=<ltrb 25 -45 25 0>}}
  }
  ScaleBox_0 {widgetClassPath=&ScaleBox, parent=SizeBox_2, widget=ScaleBox_0, properties {stretchDirection=DownOnly}, slot {class=SizeBoxSlot}}
  Text_WindowTittle {
   widgetClassPath=&CommonTextBlock
   bIsVariable=true
   parent=ScaleBox_0
   widget=Text_WindowTittle
   properties {
    text=MATCH HISTORY
    font {
     bForceMonospaced=false
     bMaterialIsStencil=false
     fontMaterial=null
     fontObject=&Burbank
     letterSpacing=0
     monospacedWidth=1
     size=72
     skewAmount=0
     typefaceFontName=Burbank-Black
     outlineSettings {
      bApplyOutlineToDropShadows=false
      bMiteredCorners=false
      bSeparateFillAlpha=false
      outlineMaterial=null
      outlineSize=0
      outlineColor=<rgba 0 0.024 0.2552 0.9>
     }
    }
    shadowColorAndOpacity=<rgba 0 0 0 0.3>
    shadowOffset=<6 2>
   }
   slot {class=ScaleBoxSlot, properties {horizontalAlignment=HAlign_Fill, verticalAlignment=VAlign_Fill}}
  }
  SizeBox_1 {
   widgetClassPath=&SizeBox
   parent=VerticalBox_0
   widget=SizeBox_1
   properties {heightOverride=750, widthOverride=31549.3516}
   slot {class=VerticalBoxSlot}
  }
  VerticalBox_1 {
   widgetClassPath=&VerticalBox
   bIsVariable=true
   parent=SizeBox_1
   widget=VerticalBox_1
   slot {class=SizeBoxSlot, properties {padding=<ltrb 80 0 80 0>}}
  }
  HorizontalBox_109 {widgetClassPath=&HorizontalBox, parent=VerticalBox_1, widget=HorizontalBox_109, slot {class=VerticalBoxSlot}}
  SizeBox_7 {
   widgetClassPath=&SizeBox
   parent=HorizontalBox_109
   widget=SizeBox_7
   properties {bOverride_WidthOverride=true, widthOverride=150}
   slot {class=HorizontalBoxSlot, properties {verticalAlignment=VAlign_Center, padding=<ltrb 0 0 25 0>}}
  }
  ScaleBox_3 {widgetClassPath=&ScaleBox, parent=SizeBox_7, widget=ScaleBox_3, properties {stretchDirection=DownOnly}, slot {class=SizeBoxSlot}}
  Text_EventLabel {
   widgetClassPath=&CommonTextBlock
   bIsVariable=true
   parent=ScaleBox_3
   widget=Text_EventLabel
   properties {
    scrollStyle=&TextScrollStyle-Base_C
    style=/FortUILibrary/Styles/Typography/Foundation/TextStyle-Header-M.TextStyle-Header-M_C
    text="Event: "
    textTransformPolicy=ToUpper
    font {
     bForceMonospaced=false
     bMaterialIsStencil=false
     fontMaterial=null
     fontObject=&Burbank
     letterSpacing=0
     monospacedWidth=1
     size=32
     skewAmount=0
     typefaceFontName=Burbank-Black
     outlineSettings {
      bApplyOutlineToDropShadows=false
      bMiteredCorners=false
      bSeparateFillAlpha=false
      outlineMaterial=null
      outlineSize=0
      outlineColor=<rgba 0 0 0 1>
     }
    }
    shadowOffset=<0 0>
   }
   slot {class=ScaleBoxSlot, properties {horizontalAlignment=HAlign_Fill}}
  }
  ScaleBox_4 {
   widgetClassPath=&ScaleBox
   parent=HorizontalBox_109
   widget=ScaleBox_4
   properties {stretchDirection=DownOnly}
   slot {class=HorizontalBoxSlot, properties {verticalAlignment=VAlign_Center, size {sizeRule=Fill, value=1}}}
  }
  EventMatchHistoryTournamentInfo {
   widgetClassPath=/Game/UI/Events/Leaderboards/EventMatchHistoryTournamentInfo
   bIsVariable=true
   parent=ScaleBox_4
   widget=EventMatchHistoryTournamentInfo
   slot {class=ScaleBoxSlot, properties {horizontalAlignment=HAlign_Fill, verticalAlignment=VAlign_Fill}}
  }
  HorizontalBox_0 {widgetClassPath=&HorizontalBox, parent=VerticalBox_1, widget=HorizontalBox_0, slot {class=VerticalBoxSlot}}
  SizeBox_6 {
   widgetClassPath=&SizeBox
   parent=HorizontalBox_0
   widget=SizeBox_6
   properties {bOverride_MaxDesiredWidth=true, bOverride_WidthOverride=true, maxDesiredWidth=900, widthOverride=150}
   slot {class=HorizontalBoxSlot, properties {verticalAlignment=VAlign_Center, padding=<ltrb 0 0 25 0>}}
  }
  ScaleBox_2 {widgetClassPath=&ScaleBox, parent=SizeBox_6, widget=ScaleBox_2, properties {stretchDirection=DownOnly}, slot {class=SizeBoxSlot}}
  Text_TeamLabel {
   widgetClassPath=&CommonTextBlock
   bIsVariable=true
   parent=ScaleBox_2
   widget=Text_TeamLabel
   properties {
    scrollStyle=&TextScrollStyle-Base_C
    style=&TextStyle-Header-M_C
    text=Team:
    textTransformPolicy=ToUpper
    font {
     bForceMonospaced=false
     bMaterialIsStencil=false
     fontMaterial=null
     fontObject=&Burbank
     letterSpacing=0
     monospacedWidth=1
     size=32
     skewAmount=0
     typefaceFontName=Burbank-Black
     outlineSettings {
      bApplyOutlineToDropShadows=false
      bMiteredCorners=false
      bSeparateFillAlpha=false
      outlineMaterial=null
      outlineSize=0
      outlineColor=<rgba 0 0 0 1>
     }
    }
    shadowOffset=<0 0>
   }
   slot {class=ScaleBoxSlot, properties {horizontalAlignment=HAlign_Fill}}
  }
  SizeBox_5 {
   widgetClassPath=&SizeBox
   parent=HorizontalBox_0
   widget=SizeBox_5
   properties {bOverride_MaxDesiredWidth=true, maxDesiredWidth=900}
   slot {class=HorizontalBoxSlot, properties {verticalAlignment=VAlign_Center, size {sizeRule=Fill, value=1}}}
  }
  Text_TeamNames {
   widgetClassPath=&CommonTextBlock
   bInherited=true
   bIsVariable=true
   parent=SizeBox_5
   widget=Text_TeamNames
   properties {
    scrollStyle=&TextScrollStyle-Base_C
    text="[ph] Ipsum, Dolor, It_Epithet"
    textTransformPolicy=ToUpper
    font {
     bForceMonospaced=false
     bMaterialIsStencil=false
     fontMaterial=null
     fontObject=&BurbankSmall
     letterSpacing=0
     monospacedWidth=1
     size=22
     skewAmount=0
     typefaceFontName=Bold
     outlineSettings {
      bApplyOutlineToDropShadows=false
      bMiteredCorners=false
      bSeparateFillAlpha=false
      outlineMaterial=null
      outlineSize=0
      outlineColor=<rgba 0 0 0 1>
     }
    }
    shadowOffset=<0 0>
   }
   slot {class=SizeBoxSlot}
  }
  SizeBox_4 {
   widgetClassPath=&SizeBox
   parent=VerticalBox_1
   widget=SizeBox_4
   properties {bOverride_MaxDesiredHeight=true, maxDesiredHeight=700}
   slot {class=VerticalBoxSlot, properties {padding=<ltrb 0 0 0 25>}}
  }
  ListView_MatchHistory {
   widgetClassPath=&CommonListView
   bInherited=true
   bIsVariable=true
   parent=SizeBox_4
   widget=ListView_MatchHistory
   properties {
    entryWidgetClass=&/EventTeamMatchHistoryEntry.EventTeamMatchHistoryEntry_C
    navigation=&self:WidgetTree.ListView_MatchHistory.WidgetNavigation_0
    numDesignerPreviewEntries=20
   }
   slot {class=SizeBoxSlot}
  }
  SizeBox_3 {widgetClassPath=&SizeBox, parent=VerticalBox_0, widget=SizeBox_3, properties {heightOverride=50}, slot {class=VerticalBoxSlot}}
  CommonBorder {
   widgetClassPath=&CommonBorder
   parent=SizeBox_3
   widget=CommonBorder
   properties {
    verticalAlignment=VAlign_Top
    background {
     bIsDynamicallyLoaded=false
     drawAs=Image
     imageType=NoImage
     mirroring=NoMirror
     resourceName=None
     resourceObject=null
     tiling=NoTile
     imageSize=<32 32>
     margin=<ltrb 0 0 0 0>
     outlineSettings {
      bUseBrushTransparency=false
      roundingType=HalfHeightRadius
      width=0
      color {colorUseRule=UseColor_Specified, specifiedColor=<rgba 0 0 0 0>}
      cornerRadii=<0 0 0 0>
     }
     tintColor {colorUseRule=UseColor_Specified, specifiedColor=<rgba 1 1 1 1>}
     uVRegion {bIsValid=false, max=<0 0>, min=<0 0>}
    }
    brushColor=<rgba 0 0.05 0.2 1>
    padding=<ltrb 0 10 15 10>
   }
   slot {class=SizeBoxSlot}
  }
  Overlay_2 {
   widgetClassPath=&Overlay
   parent=CommonBorder
   widget=Overlay_2
   slot {class=BorderSlot, properties {verticalAlignment=VAlign_Top, padding=<ltrb 0 10 15 10>}}
  }
  CloseMainPanel {
   widgetClassPath=&IconTextButton
   bIsVariable=true
   parent=Overlay_2
   widget=CloseMainPanel
   properties {
    buttonText=CLOSE
    onButtonBaseClicked=(&self:WidgetTree.CloseMainPanel.HandleOnButtonClicked)
    onButtonBaseDoubleClicked=(&self:WidgetTree.CloseMainPanel.HandleOnButtonDoubleClicked)
    onButtonBaseHovered=(&self:WidgetTree.CloseMainPanel.HandleOnButtonHovered)
    onButtonBaseUnhovered=(&self:WidgetTree.CloseMainPanel.HandleOnButtonUnhovered)
    onSelectedChangedBase=(&self:WidgetTree.CloseMainPanel.HandleOnSelectedChanged)
    style=/FortUILibrary/Foundation/Buttons/ButtonStyles/Skew/ButtonStyle-Skew_LessTransparent.ButtonStyle-Skew_LessTransparent_C
    triggeringInputAction {dataTable=/FortUILibrary/Input/CreativeInputActionDataTable, rowName=Close}
   }
   slot {class=OverlaySlot, properties {horizontalAlignment=HAlign_Right}}
  }
  HorizontalBox_2 {
   widgetClassPath=&HorizontalBox
   parent=Overlay_2
   widget=HorizontalBox_2
   properties {bIsEnabled=false}
   slot {class=OverlaySlot, properties {verticalAlignment=VAlign_Center, padding=<ltrb 10 0 10 0>}}
  }
  Widget_AltStat0 {
   widgetClassPath=/Game/UI/Events/Leaderboards/EventTeamMatchHistoryStat
   bInherited=true
   parent=HorizontalBox_2
   widget=Widget_AltStat0
   properties {bIsEnabled=false, icon Text=1}
   slot {class=HorizontalBoxSlot, properties {padding=<ltrb 0 0 15 0>}}
  }
  Widget_AltStat1 {
   widgetClassPath=&EventTeamMatchHistoryStat
   bInherited=true
   parent=HorizontalBox_2
   widget=Widget_AltStat1
   properties {bIsEnabled=false, icon Text=2}
   slot {class=HorizontalBoxSlot, properties {padding=<ltrb 15 0 15 0>}}
  }
  Widget_AltStat2 {
   widgetClassPath=&EventTeamMatchHistoryStat
   bInherited=true
   parent=HorizontalBox_2
   widget=Widget_AltStat2
   properties {bIsEnabled=false, icon Text=3}
   slot {class=HorizontalBoxSlot, properties {padding=<ltrb 15 0 15 0>}}
  }
  Widget_AltStat3 {
   widgetClassPath=&EventTeamMatchHistoryStat
   bInherited=true
   parent=HorizontalBox_2
   widget=Widget_AltStat3
   properties {bIsEnabled=false, icon Text=4}
   slot {class=HorizontalBoxSlot, properties {padding=<ltrb 15 0 0 0>}}
  }
 ]
}
