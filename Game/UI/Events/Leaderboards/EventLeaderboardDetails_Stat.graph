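# Event leaderboard details row widget that shows one stat's name and numeric value with an animated reveal.
# SetStatData fills both texts, formats the value to the requested fractional digits, then calls EventStatAnim.
# That event snaps UpdateStat to its initial state with a zero-length reverse play, waits on a retriggerable delay staggered by StatItemIndex (longer on gamepad), then plays it forward to fade and scale-punch the value in.
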WidgetBlueprint /Game/UI/Events/Leaderboards/EventLeaderboardDetails_Stat
animations [UpdateStat {
  binding_count=1
  length_seconds=0.1799
  object=&self:UpdateStat
  sequence {
   bindings [Text_StatValue {
     id=#1
     possessed_class=CommonTextBlock
     tracks [
      {
       type=MovieSceneFloatTrack
       display_name=Render Opacity
       sections [{type=MovieSceneFloatSection, channels [None {type=MovieSceneScriptingFloatChannel, keys [[0,0],[10.796,1]]}]}]
      }
      {
       type=MovieScene2DTransformTrack
       sections [{
         type=MovieScene2DTransformSection
         channels [
          Translation.X {type=MovieSceneScriptingFloatChannel}
          Translation.Y {type=MovieSceneScriptingFloatChannel}
          Angle {type=MovieSceneScriptingFloatChannel}
          Scale.X {type=MovieSceneScriptingFloatChannel, keys [[0,2.1525],[8.228,1.0597],[10.053,1.2554],[10.617,1.2]]}
          Scale.Y {type=MovieSceneScriptingFloatChannel, keys [[0,2.1525],[8.228,1.0597],[10.053,1.2554],[10.617,1.2]]}
          Shear.X {type=MovieSceneScriptingFloatChannel}
          Shear.Y {type=MovieSceneScriptingFloatChannel}
         ]
        }]
      }
     ]
    }]
   display_rate {numerator=60}
  }
 }]
graphs [
 EventGraph {nodes [
   EventStatAnim() {is=title K2Node_CustomEvent}
   n1 = Get UpdateStat() -> (UpdateStat:Widget Animation Object Reference)
   n2 = GetEndTime(self=@n1)
   > PlayAnimationTimeRange!(
    execute
    self:User Widget Object Reference
    InAnimation:Widget Animation Object Reference=@n1
    StartAtTime:float=@n2
    EndAtTime:float=@n2
    NumLoopsToPlay:int=1
    PlayMode:EUMGSequencePlayMode Enum=Reverse
    PlaybackSpeed:float=1
    bRestoreState:bool=false
   ) -> (then, ReturnValue:Widget Animation Handle Structure)
   n3 = Get StatItemIndex() -> (StatItemIndex:int=0)
   n4 = IsUsingGamepad!(self:Fort Global UIContext Object Reference=Get FortGlobalUIContext!() -> (ReturnValue:Fort Global UIContext Object Reference) {is=title K2Node_GetBlueprintContextEx}) -> (ReturnValue:bool)
   > RetriggerableDelay(Duration=float + float(
     A=int * float!(A:int=@n3, B:double=0.1) -> (ReturnValue:double)
     B=Select(Option 0:double=0, Option 1:double=0.2, Index=@n4) -> (ReturnValue:double)
    ) {is=title K2Node_CommutativeAssociativeBinaryOperator})
   > PlayAnimationForward(InAnimation=@n1)
   Event PreConstruct()
   Event Construct()
   Event Tick()
  ]}
 SetStatData {nodes [
   n1 = SetStatData!() -> (then, Name:Text, Value:double=~, NumFractionalDigits:int) {is=title K2Node_FunctionEntry}
   > "SetText (Text)"(self=Get Text_StatName() -> (Text_StatName:Common Text Object Reference), InText=@n1)
   > "SetText (Text)"(
    self=Get Text_StatValue() -> (Text_StatValue:Common Text Object Reference)
    InText="To Text (Float)"(
     Value=@n1.Value
     MinimumIntegralDigits=1
     MaximumIntegralDigits=324
     MinimumFractionalDigits=@n1.NumFractionalDigits
     MaximumFractionalDigits=@n1.NumFractionalDigits
    )
   )
   > EventStatAnim() {is=title K2Node_CallFunction}
  ]}
]
variables [StatItemIndex {type=integer}]
widgets {
 info {inheritedWidgetCount=0, parentClass=&UserWidget, rootWidgetClass=&HorizontalBox, widgetCount=5}
 widgets [
  HorizontalBox_0 {widgetClassPath=&HorizontalBox, widget=HorizontalBox_0}
  Text_StatName {
   widgetClassPath=&CommonTextBlock
   bIsVariable=true
   parent=HorizontalBox_0
   widget=Text_StatName
   properties {
    text=Games Played
    font {
     bForceMonospaced=false
     bMaterialIsStencil=false
     fontMaterial=null
     fontObject=&Burbank
     letterSpacing=0
     monospacedWidth=1
     size=24
     skewAmount=0
     typefaceFontName=Burbank-PlayerPower
     outlineSettings {
      bApplyOutlineToDropShadows=false
      bMiteredCorners=false
      bSeparateFillAlpha=false
      outlineMaterial=null
      outlineSize=0
      outlineColor=<rgba 0 0 0 1>
     }
    }
    shadowOffset=<0 0>
   }
   slot {class=HorizontalBoxSlot, properties {verticalAlignment=VAlign_Center, size {sizeRule=Fill, value=1}}}
  }
  SizeBox_1 {
   widgetClassPath=&SizeBox
   parent=HorizontalBox_0
   widget=SizeBox_1
   properties {
    bOverride_HeightOverride=true
    bOverride_MaxDesiredWidth=true
    bOverride_MinDesiredWidth=true
    heightOverride=50
    maxDesiredWidth=250
    minDesiredWidth=250
   }
   slot {class=HorizontalBoxSlot, properties {horizontalAlignment=HAlign_Center, verticalAlignment=VAlign_Center}}
  }
  ScaleBox_1 {widgetClassPath=&ScaleBox, parent=SizeBox_1, widget=ScaleBox_1, properties {stretchDirection=DownOnly}, slot {class=SizeBoxSlot}}
  Text_StatValue {
   widgetClassPath=&CommonTextBlock
   bIsVariable=true
   parent=ScaleBox_1
   widget=Text_StatValue
   properties {
    text=10
    font {
     bForceMonospaced=false
     bMaterialIsStencil=false
     fontMaterial=null
     fontObject=&Burbank
     letterSpacing=0
     monospacedWidth=1
     size=30
     skewAmount=0
     typefaceFontName=BurbankRegular-Black
     outlineSettings {
      bApplyOutlineToDropShadows=false
      bMiteredCorners=false
      bSeparateFillAlpha=false
      outlineMaterial=null
      outlineSize=0
      outlineColor=<rgba 0 0 0 1>
     }
    }
    renderTransformPivot=<0.5 0.9>
    shadowOffset=<0 0>
   }
   slot {class=ScaleBoxSlot}
  }
 ]
}
