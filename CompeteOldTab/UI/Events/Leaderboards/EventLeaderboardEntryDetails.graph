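# Tournament leaderboard details panel, a FortLeaderboardEntryDetails widget showing the selected entry's team names, score and stats.
# Show Details drives a dynamic-material pie splitting placement and elimination points, plus stat rows; alternative tracked stats hide the pie and replace some rows.
# EventColorize applies tournament colors; EventIntro replays the intro, using a retriggerable delay on gamepad.
# Buttons open career stats, a team profile selector or match history, with runtime options gating profiles.
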WidgetBlueprint /CompeteOldTab/UI/Events/Leaderboards/EventLeaderboardEntryDetails
animations [
 Intro {
  binding_count=10
  length_seconds=0.55
  object=&self:Intro
  sequence {
   bindings [
    VerticalBox_2 {
     id=#1
     possessed_class=VerticalBox
     tracks [{
       type=MovieSceneFloatTrack
       display_name=Render Opacity
       sections [{type=MovieSceneFloatSection, channels [None {type=MovieSceneScriptingFloatChannel, keys [[13.361,0],[28.999,1]]}]}]
      }]
    }
    CommonBorder_LeftStat {
     id=#2
     possessed_class=CommonBorder
     tracks [{
       type=MovieSceneFloatVectorTrack
       display_name=Desired Size Scale
       sections [{
         type=MovieSceneFloatVectorSection
         channels [
          Vector.X {type=MovieSceneScriptingFloatChannel, keys [[20.051,0],[29.392,1]]}
          Vector.Y {type=MovieSceneScriptingFloatChannel}
         ]
        }]
      }]
    }
    Image_94 {
     id=#3
     possessed_class=Image
     tracks [{
       type=MovieScene2DTransformTrack
       sections [{
         type=MovieScene2DTransformSection
         channels [
          Translation.X {type=MovieSceneScriptingFloatChannel}
          Translation.Y {type=MovieSceneScriptingFloatChannel}
          Angle {type=MovieSceneScriptingFloatChannel}
          Scale.X {type=MovieSceneScriptingFloatChannel, keys [[29.861,0],[33,1]]}
          Scale.Y {type=MovieSceneScriptingFloatChannel, keys [[29.861,0],[33,1]]}
          Shear.X {type=MovieSceneScriptingFloatChannel}
          Shear.Y {type=MovieSceneScriptingFloatChannel}
         ]
        }]
      }]
    }
    CommonTextBlock_1 {
     id=#4
     possessed_class=CommonTextBlock
     tracks [{
       type=MovieSceneFloatTrack
       display_name=Render Opacity
       sections [{type=MovieSceneFloatSection, channels [None {type=MovieSceneScriptingFloatChannel, keys [[1.675,0],[8.342,1]]}]}]
      }]
    }
    Image_Pie {
     id=#5
     possessed_class=Image
     tracks [{
       type=MovieSceneWidgetMaterialTrack
       sections [{
         type=MovieSceneParameterSection
         channels [Reveal {type=MovieSceneScriptingFloatChannel, keys [[0,0],[19.919,0.9946],[23.399,1.0286],[25.419,1]]}]
        }]
      }]
    }
    CommonBorder_RightStat {
     id=#6
     possessed_class=CommonBorder
     tracks [{
       type=MovieSceneFloatVectorTrack
       display_name=Desired Size Scale
       sections [{
         type=MovieSceneFloatVectorSection
         channels [
          Vector.X {type=MovieSceneScriptingFloatChannel, keys [[24.252,0],[29.861,1]]}
          Vector.Y {type=MovieSceneScriptingFloatChannel}
         ]
        }]
      }]
    }
    CommonTextBlock_Score {
     id=#7
     possessed_class=CommonTextBlock
     tracks [
      {
       type=MovieSceneFloatTrack
       display_name=Render Opacity
       sections [{type=MovieSceneFloatSection, channels [None {type=MovieSceneScriptingFloatChannel, keys [[16.055,0],[28.999,1]]}]}]
      }
      {
       type=MovieScene2DTransformTrack
       sections [{
         type=MovieScene2DTransformSection
         channels [
          Translation.X {type=MovieSceneScriptingFloatChannel}
          Translation.Y {type=MovieSceneScriptingFloatChannel}
          Angle {type=MovieSceneScriptingFloatChannel}
          Scale.X {type=MovieSceneScriptingFloatChannel, keys [[14.263,0.5],[25.419,1.0279],[29.861,1]]}
          Scale.Y {type=MovieSceneScriptingFloatChannel, keys [[14.263,0.5],[25.419,1.0279],[29.861,1]]}
          Shear.X {type=MovieSceneScriptingFloatChannel}
          Shear.Y {type=MovieSceneScriptingFloatChannel}
         ]
        }]
      }
     ]
    }
    Image_58 {
     id=#8
     possessed_class=Image
     tracks [{
       type=MovieScene2DTransformTrack
       sections [{
         type=MovieScene2DTransformSection
         channels [
          Translation.X {type=MovieSceneScriptingFloatChannel}
          Translation.Y {type=MovieSceneScriptingFloatChannel}
          Angle {type=MovieSceneScriptingFloatChannel}
          Scale.X {type=MovieSceneScriptingFloatChannel, keys [[26.641,0],[28.999,1]]}
          Scale.Y {type=MovieSceneScriptingFloatChannel, keys [[26.641,0],[28.999,1]]}
          Shear.X {type=MovieSceneScriptingFloatChannel}
          Shear.Y {type=MovieSceneScriptingFloatChannel}
         ]
        }]
      }]
    }
    VerticalBox_Placement {
     id=#9
     possessed_class=VerticalBox
     tracks [{
       type=MovieSceneFloatTrack
       display_name=Render Opacity
       sections [{type=MovieSceneFloatSection, channels [None {type=MovieSceneScriptingFloatChannel, keys [[13.361,0],[28.999,1]]}]}]
      }]
    }
    Image_PieShadow {
     id=#10
     possessed_class=Image
     tracks [{
       type=MovieSceneWidgetMaterialTrack
       sections [{
         type=MovieSceneParameterSection
         channels [Reveal {type=MovieSceneScriptingFloatChannel, keys [[0,0],[19.919,0.9946],[23.399,1.0286],[25.419,1]]}]
        }]
      }]
    }
   ]
   display_rate {numerator=60}
  }
 }
 ProfileHover {
  binding_count=1
  length_seconds=0.15
  object=&self:ProfileHover
  sequence {
   bindings [CommonTextBlock_ProfileText {
     id=#11
     possessed_class=CommonTextBlock
     tracks [{
       type=MovieSceneColorTrack
       display_name=Color and Opacity
       sections [{
         type=MovieSceneColorSection
         end_frame=9
         start_frame=0
         channels [
          Color.R {type=MovieSceneScriptingFloatChannel, keys [[0,1],[9,0]]}
          Color.G {type=MovieSceneScriptingFloatChannel, keys [[0,1],[9,0]]}
          Color.B {type=MovieSceneScriptingFloatChannel, keys [[0,1],[9,0]]}
          Color.A {type=MovieSceneScriptingFloatChannel, const [1,2]}
         ]
        }]
      }]
    }]
   display_rate {numerator=60}
  }
 }
 MatchHistoryHover {
  binding_count=1
  length_seconds=0.15
  object=&self:MatchHistoryHover
  sequence {
   bindings [CommonTextBlock_MatchHistoryTextMatchHistory {
     id=#12
     possessed_class=CommonTextBlock
     tracks [{
       type=MovieSceneColorTrack
       display_name=Color and Opacity
       sections [{
         type=MovieSceneColorSection
         end_frame=9
         start_frame=0
         channels [
          Color.R {type=MovieSceneScriptingFloatChannel, keys [[0,1],[9,0]]}
          Color.G {type=MovieSceneScriptingFloatChannel, keys [[0,1],[9,0]]}
          Color.B {type=MovieSceneScriptingFloatChannel, keys [[0,1],[9,0]]}
          Color.A {type=MovieSceneScriptingFloatChannel, const [1,2]}
         ]
        }]
      }]
    }]
   display_rate {numerator=60}
  }
 }
]
graphs [
 EventGraph {nodes [
   n1 = EventColorize!() -> (OutputDelegate, then, TournamentDisplayInfo:Fort Tournament Display Info Structure) {is=title K2Node_CustomEvent}
   > n2 = Set CurrentDisplayInfo(CurrentDisplayInfo:Fort Tournament Display Info Structure=@n1.TournamentDisplayInfo) -> (Output_Get:Fort Tournament Display Info Structure)
   n3 = Break Fort Tournament Display Info!(FortTournamentDisplayInfo:"Fort Tournament Display Info Structure (by ref)"=@n2) -> (
    TitleLine1:Text
    TitleLine2:Text
    ScheduleInfo:Text
    FlavorDescription:Text
    DetailsDescription:Text
    ShortFormatTitle:Text
    LongFormatTitle:Text
    BackgroundTitle:Text
    PinScoreRequirement:int
    PinEarnedText:Text
    BaseColor:Linear Color Structure
    PrimaryColor:Linear Color Structure
    SecondaryColor:Linear Color Structure
    HighlightColor:Linear Color Structure
    TitleColor:Linear Color Structure
    ShadowColor:Linear Color Structure
    BackgroundLeftColor:Linear Color Structure
    BackgroundRightColor:Linear Color Structure
    BackgroundTextColor:Linear Color Structure
    PosterFadeColor:Linear Color Structure
    AlertText:Text
    SeriesPointLeaderboardName:Text
    AlertType:EFortTournamentAlertType Enum
    RoundDisplayNames:Array of Texts
   ) {is=title K2Node_BreakStruct}
   SetColorAndOpacity(
    self:Image Object Reference=Get Image_PieShadow() -> (Image_PieShadow:Image Object Reference)
    InColorAndOpacity:Linear Color Structure=@n3.ShadowColor
   )
   > n4 = SetShadowColorAndOpacity!(
    execute
    self:Text Object Reference=Get CommonTextBlock_Score() -> (CommonTextBlock_Score:Common Text Object Reference)
    InShadowColorAndOpacity:Linear Color Structure=@n3.ShadowColor
   ) -> (then)
   n5 = Get CommonTextBlock_PlayerNamesTitle() -> (CommonTextBlock_PlayerNamesTitle:Common Text Object Reference)
   SetColorAndOpacity(
    execute=[@n4.then, @n26.then]
    self:Text Object Reference=@n5
    InColorAndOpacity:Slate Color Structure=Make Slate Color(SpecifiedColor=@n3.PrimaryColor)
   )
   > Set Outline Color!(In:Exec, CommonTextWidget:Common Text Object Reference=@n5, Color:Linear Color Structure=@n3.ShadowColor) -> (Out:Exec) {is=title K2Node_MacroInstance}
   > SetColorAndOpacity(
    self=Get CommonTextBlock_PlayerNames() -> (CommonTextBlock_PlayerNames:Common Text Object Reference)
    InColorAndOpacity=Make Slate Color(SpecifiedColor=@n3.PrimaryColor)
   )
   "On Button Hovered (Button_Profiles)"!() -> (OutputDelegate, then, Button:Common Button Legacy Object Reference) {is=title K2Node_ComponentBoundEvent}
   n6 = Get ProfileHover() -> (ProfileHover:Widget Animation Object Reference)
   > PlayAnimationForward(InAnimation=@n6)
   n7 = "On Button Clicked (Button_Profiles)"!() -> (OutputDelegate, then, Button:Common Button Legacy Object Reference) {is=title K2Node_ComponentBoundEvent}
   n8 = Branch(
    execute=[@n7.then, @n20.then]
    Condition=Get bDisableViewOtherProfilesFromCompLeaderboards!(self:Fort Runtime Options Object Reference=GetRuntimeOptions()) -> (bDisableViewOtherProfilesFromCompLeaderboards:bool)
   )
   Branch(execute=@n8.else, Condition=Get bAllowProfileNavigation() -> (bAllowProfileNavigation:bool))
   n9 = Get LeaderboardEntryData() -> (LeaderboardEntryData:Fort Event Leaderboard Entry Data Object Reference)
   > n10 = Branch(Condition="Equal (Integer)"(A=GetNumTeammates!(self:Fort Event Leaderboard Entry Data Object Reference=@n9) -> (ReturnValue:int=0), B=1) {is=title K2Node_CallFunction})
   n11 = GetTeamAccountId!(self:Fort Event Leaderboard Entry Data Object Reference=@n9, Index:int=0) -> (ReturnValue:String)
   n12 = GetPlayerDisplayName!(self:Fort Event Leaderboard Entry Data Object Reference=@n9, AccountId:String=@n11) -> (ReturnValue:String)
   > n13 = Branch(Condition=IsEmpty(InString=@n12))
   ShowCareerStatsScreen!(
    execute=@n13.else
    self:Fort Leaderboard Entry Details Object Reference
    Params:"Career Stats Screen Params Structure (by ref)"=Make Career Stats Screen Params!(
     ContextPC:Player Controller Object Reference=GetPlayerController!(PlayerIndex:int=0) -> (ReturnValue:Player Controller Object Reference)
     PlayerAccountId:String=@n11
     PlayerDisplayName:Text="To Text (String)"(InString=@n12)
     GameMode:String=br
     bIsRanked:bool=True
     SeasonTrack:String
     LimitedGameModeIndex:int=-1
    ) -> (CareerStatsScreenParams:Career Stats Screen Params Structure) {is=title K2Node_MakeStruct}
   ) -> (then)
   n14 = Create Widget!(
    execute:Exec=@n10.else
    Class:User Widget Class Reference=&/EventPlayerStatProfileSelector.EventPlayerStatProfileSelector_C
    OwningPlayer:Player Controller Object Reference=GetOwningPlayer()
   ) -> (then:Exec, ReturnValue:Event Player Stat Profile Selector Object Reference)
   > SetProfilesByLeaderboardData!(
    execute
    self:Fort Event Player Stat Profile Selector Object Reference=@n14
    DataForDisplay:Fort Event Leaderboard Entry Data Object Reference=Get LeaderboardEntryData()
   ) -> (then)
   > Set Up Profile Selector Style!(
    execute
    self:Event Player Stat Profile Selector Object Reference=@n14
    Tournament Display Info:Fort Tournament Display Info Structure=Get CurrentDisplayInfo() -> (CurrentDisplayInfo:Fort Tournament Display Info Structure)
   ) -> (then) {is=title K2Node_CallFunction}
   > TryPushWidgetAsModalPanel!(execute, PushWidget:User Widget Object Reference=@n14) -> (then)
   "On Button Unhovered (Button_Profiles)"!() -> (OutputDelegate, then, Button:Common Button Legacy Object Reference) {is=title K2Node_ComponentBoundEvent}
   > PlayAnimationReverse(InAnimation=@n6)
   "On Button Clicked (Button_MatchHistory)"!() -> (OutputDelegate, then, Button:Common Button Legacy Object Reference) {is=title K2Node_ComponentBoundEvent}
   > n15 = Create Widget(Class=&/EventTeamMatchHistoryModal.EventTeamMatchHistoryModal_C, OwningPlayer=GetOwningPlayer()) -> (ReturnValue:Event Team Match History Modal Object Reference)
   > SetupMatchHistory!(
    execute
    self:Fort Event Team Match History Modal Object Reference=@n15
    DataForDisplay:Fort Event Leaderboard Entry Data Object Reference=Get LeaderboardEntryData()
    TournamentId:String=Get TournamentId() -> (TournamentId:String)
    EventWindowId:String=Get EventWindowId() -> (EventWindowId:String)
    LeaderboardGroupingKey:String=Get LeaderboardGroupingKey() -> (LeaderboardGroupingKey:String)
    LeaderboardInstanceId:String=Get LeaderboardInstanceId() -> (LeaderboardInstanceId:String)
   ) -> (then)
   > Set Up Match History Style!(
    execute
    self:Event Team Match History Modal Object Reference=@n15
    Tournament Display Info:Fort Tournament Display Info Structure=Get CurrentDisplayInfo()
   ) -> (then) {is=title K2Node_CallFunction}
   > TryPushWidgetAsModalPanel(PushWidget=@n15)
   "On Button Hovered (Button_MatchHistory)"!() -> (OutputDelegate, then, Button:Common Button Legacy Object Reference) {is=title K2Node_ComponentBoundEvent}
   n16 = Get MatchHistoryHover() -> (MatchHistoryHover:Widget Animation Object Reference)
   > PlayAnimationForward(InAnimation=@n16)
   "On Button Unhovered (Button_MatchHistory)"!() -> (OutputDelegate, then, Button:Common Button Legacy Object Reference) {is=title K2Node_ComponentBoundEvent}
   > PlayAnimationReverse(InAnimation=@n16)
   EventIntro() {is=title K2Node_CustomEvent}
   n17 = Get Intro()
   n18 = GetEndTime(self=@n17)
   > PlayAnimationTimeRange!(
    execute
    self:User Widget Object Reference
    InAnimation:Widget Animation Object Reference=@n17
    StartAtTime:float=@n18
    EndAtTime:float=@n18
    NumLoopsToPlay:int=1
    PlayMode:EUMGSequencePlayMode Enum=Reverse
    PlaybackSpeed:float=1
    bRestoreState:bool=false
   ) -> (then, ReturnValue:Widget Animation Handle Structure)
   n19 = Get FortGlobalUIContext!() -> (ReturnValue:Fort Global UIContext Object Reference) {is=title K2Node_GetBlueprintContextEx}
   > RetriggerableDelay(Duration=Select(Option 0:double=0, Option 1:double=0.2, Index=IsUsingGamepad!(self:Fort Global UIContext Object Reference=@n19) -> (ReturnValue:bool)) -> (ReturnValue:double))
   > PlayAnimationTimeRange(InAnimation=@n17, StartAtTime=0, EndAtTime=0, NumLoopsToPlay=1, PlaybackSpeed=1, bRestoreState=false)
   n20 = LeaderboardEntryDoubleClicked() {is=title K2Node_CustomEvent}
   Event Construct()
   n21 = Get VerticalBox_StatContainer() -> (VerticalBox_StatContainer:Vertical Box Object Reference)
   > n22 = For Loop(LastIndex=GetChildrenCount!(self:Panel Widget Object Reference=@n21) -> (ReturnValue:int=0))
   > n23 = Cast To EventLeaderboardDetails_Stat!(execute, Object=GetChildAt!(self:Panel Widget Object Reference=@n21, Index:int=@n22) -> (ReturnValue:Widget Object Reference)) -> (then, CastFailed, AsEvent Leaderboard Details Stat:Event Leaderboard Details Stat Object Reference) {is=title K2Node_DynamicCast}
   > Set StatItemIndex!(execute, StatItemIndex:int=@n22, self:Event Leaderboard Details Stat Object Reference=@n23) -> (then, Output_Get:int=0)
   > Subtract Safezone!(In:Exec, OverlaySlot:Widget Object Reference=Get Image_DetailsBG() -> (Image_DetailsBG:Image Object Reference), ScaleUp:bool) -> (Out:Exec) {is=title K2Node_MacroInstance}
   > Branch(Condition=Get bDisableViewOtherProfilesFromCompLeaderboards(self=GetRuntimeOptions()))
   > SetVisibility(self=Get Button_Profiles() -> (Button_Profiles:Icon Text Button Object Reference), InVisibility=Collapsed)
   Reroute Node(InputPin:Linear Color Structure=@n3.SecondaryColor) -> (OutputPin:Linear Color Structure)
   Reroute Node(InputPin=@n3.BackgroundRightColor)
   n24 = SetContentColorAndOpacity!(
    execute=@n27.then
    self:Border Object Reference=Get CommonBorder_PlacementPointer() -> (CommonBorder_PlacementPointer:Common Border Object Reference)
    InContentColorAndOpacity:Linear Color Structure=@n3.PrimaryColor
   ) -> (then)
   n25 = GetDynamicMaterial(execute=@n28.then, self=Get Image_Pie() -> (Image_Pie:Image Object Reference))
   SetVectorParameterValue(execute=@n24.then, self=@n25, ParameterName=Percent2_Color, Value=@n3.HighlightColor)
   > SetColorAndOpacity(
    self:Image Object Reference=Get Image_ElimIcon() -> (Image_ElimIcon:Image Object Reference)
    InColorAndOpacity:Linear Color Structure=@n3.HighlightColor
   )
   > SetContentColorAndOpacity(
    self=Get CommonBorder_ElimPointer() -> (CommonBorder_ElimPointer:Common Border Object Reference)
    InContentColorAndOpacity=@n3.HighlightColor
   )
   > n26 = SetColorAndOpacity(self=Get Image_PlayernameBG() -> (Image_PlayernameBG:Image Object Reference), InColorAndOpacity=@n3.ShadowColor)
   SetVectorParameterValue(execute=@n25.then, self=@n25, ParameterName=Percent1_Color, Value=@n3.PrimaryColor)
   > n27 = SetColorAndOpacity(self=Get Image_PlacementIcon() -> (Image_PlacementIcon:Image Object Reference), InColorAndOpacity=@n3.PrimaryColor)
   n28 = SetColorAndOpacity(execute=@n29.then, self=Get Image_DetailsBG(), InColorAndOpacity=@n3.BackgroundRightColor)
   n29 = SetShadowColorAndOpacity(
    execute=@n2.then
    self=Get CommonTextBlock_Title() -> (CommonTextBlock_Title:Common Text Object Reference)
    InShadowColorAndOpacity=@n3.BackgroundRightColor
   )
  ]}
 ProfileSelectedFromSelector {nodes [ProfileSelectedFromSelector() {is=title K2Node_FunctionEntry}]}
 Set Alternative Stat Widget {nodes [
   n1 = Set Alternative Stat Widget!() -> (then, Stat Widget:Event Leaderboard Details Stat Object Reference, Stats Names:Array of Strings, Stat Index:int) {is=title K2Node_FunctionEntry}
   n2 = "integer > integer"(A=Length(TargetArray:Array of Strings=@n1.Stats Names), B=@n1.Stat Index)
   > Set Visibility(
    Visibility=Select(Option 0:ESlateVisibility Enum=Collapsed, Option 1:ESlateVisibility Enum=SelfHitTestInvisible, Index=@n2) -> (ReturnValue:ESlateVisibility Enum)
    self=@n1
   )
   > Branch(Condition=@n2)
   n3 = Get LeaderboardEntryData()
   n4 = "Get (a ref)"(Array:Array of Strings=@n1.Stats Names, Dimension 1=@n1.Stat Index) -> (Output:"String (by ref)")
   n5 = GetAlternativeLeaderboardTrackedStatLabel!(self:Fort Leaderboard Entry Details Object Reference, TrackedStatName:String=@n4) -> (ReturnValue:Text)
   > SetStatData!(
    execute
    self:Event Leaderboard Details Stat Object Reference=@n1
    Name:Text=Select(Option 0:Text=@n5, Option 1:Text="To Text (String)"(InString=@n4), Index=TextIsEmpty(InText=@n5)) -> (ReturnValue:Text)
    Value:double="To Float (Integer)"(InInt=GetTotalTrackedStat!(self:Fort Event Leaderboard Entry Data Object Reference=@n3, TrackedStatName:String=@n4) -> (ReturnValue:int=0))
    NumFractionalDigits:int=0
   ) -> (then)
   > Return Node()
  ]}
 Set Leaderboard Key {nodes [
   n1 = Set Leaderboard Key!() -> (then, LeaderboardInstanceId:String, LeaderboardGroupingId:String) {is=title K2Node_FunctionEntry}
   > Set LeaderboardInstanceId(LeaderboardInstanceId:String=@n1) -> (Output_Get:String)
   > Set LeaderboardGroupingKey(LeaderboardGroupingKey:String=@n1.LeaderboardGroupingId) -> (Output_Get:String)
  ]}
 Set Points Breakdown {nodes [
   Set Points Breakdown!() -> (then, EventType:ETournamentDisplayType Enum=Default) {is=title K2Node_FunctionEntry}
   n1 = Get LeaderboardEntryData()
   n2 = Get LeaderboardEntryData()
   > Set ScoredPointsForPie(ScoredPointsForPie:int=int + int(
     A=GetTotalPointsForCategory!(self:Fort Event Leaderboard Entry Data Object Reference=@n2, ScoreCategoryId:Name=PLACEMENT_STAT_INDEX) -> (ReturnValue:int=0)
     B=GetTotalPointsForCategory(self=@n1, ScoreCategoryId=TEAM_ELIMS_STAT_INDEX) -> (ReturnValue=0)
    )) -> (Output_Get:int=0)
   n3 = Get LeaderboardEntryData()
   > "SetText (Text)"(
    self=Get CommonTextBlock_Score()
    InText="To Text (Integer)"(Value=GetScore!(self:Fort Event Leaderboard Entry Data Object Reference=@n3) -> (ReturnValue:int=0))
   )
   > n4 = GetDynamicMaterial(self=Get Image_Pie())
   > SetBrushFromMaterial(self=Get Image_PieShadow(), Material=@n4)
   n5 = GetTotalPointsForCategory(self=Get LeaderboardEntryData(), ScoreCategoryId=PLACEMENT_STAT_INDEX) -> (ReturnValue=0)
   n6 = Get ScoredPointsForPie() -> (ScoredPointsForPie:int=0)
   n7 = float * float(A=float / float(A="To Float (Integer)"(InInt=@n5), B="To Float (Integer)"(InInt=@n6)) {is=title K2Node_CallFunction}, B=100)
   n8 = float / float(A="To Float (Integer)"(InInt=Floor(A=@n7) {is=title K2Node_CallFunction}), B=100)
   n9 = Get EventType() -> (EventType:ETournamentDisplayType Enum=Default)
   > SetScalarParameterValue(
    self=@n4
    ParameterName=Percent1
    Value=SelectFloat(A=0, B=@n8, bPickA="Equal (Enum)"(A:ETournamentDisplayType Enum=@n9, B:ETournamentDisplayType Enum=Creative))
   )
   > SetVisibility(
    self=Get CommonBorder_PlacementPointer()
    InVisibility=Select(
     Option 0:ESlateVisibility Enum=HitTestInvisible
     Option 1:ESlateVisibility Enum=Hidden
     Index="float < float"!(A:double=@n8, B:double=0.1) -> (ReturnValue:bool) {is=title K2Node_CallFunction}
    ) -> (ReturnValue:ESlateVisibility Enum)
   )
   n10 = Get LeaderboardEntryData()
   > "SetText (Text)"(
    self=Get Text_PlacementPointInfo() -> (Text_PlacementPointInfo:Common Text Object Reference)
    InText=Format Text!(
     Format="LOCTEXT(\"({PointCount} PTS - {PointPercentage}%)\")"
     PointCount:int=GetTotalPointsForCategory(self=@n10, ScoreCategoryId=PLACEMENT_STAT_INDEX) -> (ReturnValue=0)
     PointPercentage:Text="To Text (Float)"(Value=@n7, MinimumIntegralDigits=1, MaximumIntegralDigits=324, MinimumFractionalDigits=0, MaximumFractionalDigits=0)
    ) -> (Result)
   )
   n11 = Get EventType() -> (EventType=Default)
   > SetVisibility(
    self=Get VerticalBox_Placement() -> (VerticalBox_Placement:Vertical Box Object Reference)
    InVisibility=Select(Option 0=SelfHitTestInvisible, Option 1=Collapsed, Index="Equal (Enum)"(A=@n11, B=Creative))
   )
   n12 = GetTotalPointsForCategory(self=Get LeaderboardEntryData(), ScoreCategoryId=TEAM_ELIMS_STAT_INDEX) -> (ReturnValue=0)
   n13 = Get ScoredPointsForPie() -> (ScoredPointsForPie=0)
   n14 = float * float(A=float / float(A="To Float (Integer)"(InInt=@n12), B="To Float (Integer)"(InInt=@n13)), B=100)
   n15 = float / float(A="To Float (Integer)"(InInt=Floor(A=@n14)), B=100)
   > SetScalarParameterValue(self=@n4, ParameterName=Percent2, Value=@n15)
   > SetVisibility(
    self=Get CommonBorder_ElimPointer()
    InVisibility=Select(
     Option 0=HitTestInvisible
     Option 1=Hidden
     Index="float <= float"!(A:double=@n15, B:double=0.3) -> (ReturnValue:bool) {is=title K2Node_CallFunction}
    )
   )
   n16 = Get LeaderboardEntryData()
   > "SetText (Text)"(
    self=Get Text_EliminationPointInfo() -> (Text_EliminationPointInfo:Common Text Object Reference)
    InText=Format Text(
     Format="LOCTEXT(\"({PointCount} PTS - {PointPercentage}%)\")"
     PointCount=GetTotalPointsForCategory(self=@n16, ScoreCategoryId=TEAM_ELIMS_STAT_INDEX) -> (ReturnValue=0)
     PointPercentage="To Text (Float)"(Value=@n14, MinimumIntegralDigits=1, MaximumIntegralDigits=324, MinimumFractionalDigits=0, MaximumFractionalDigits=0)
    )
   )
   n17 = "float <= float"(A=@n15, B=0)
   n18 = "Equal (Enum)"(A=Get EventType() -> (EventType=Default), B=Creative)
   > SetVisibility(
    self=[Get Image_Pie(), Get Image_PieShadow()]
    InVisibility=Select(Option 0=SelfHitTestInvisible, Option 1=Hidden, Index=AND Boolean(A=@n17, B=@n18))
   )
  ]}
 Set Stats Rows {nodes [
   Set Stats Rows!() -> (then, EventType:ETournamentDisplayType Enum=Default) {is=title K2Node_FunctionEntry}
   > n1 = Sequence!(execute) -> (then_0, then_1, then_2:Exec, then_3:Exec, then_4:Exec)
   n2 = Get LeaderboardEntryData()
   > SetStatData(
    self=Get Stat_MatchesPlayed() -> (Stat_MatchesPlayed:Event Leaderboard Details Stat Object Reference)
    Name="LOCTEXT(\"Matches Played\")"
    Value="To Float (Integer)"(InInt=GetTotalMatchesPlayed!(self:Fort Event Leaderboard Entry Data Object Reference=@n2) -> (ReturnValue:int=0))
    NumFractionalDigits=0
   )
   SetStatData(
    execute=@n1.then_1
    self=Get Stat_AveragePointsPerMatch() -> (Stat_AveragePointsPerMatch:Event Leaderboard Details Stat Object Reference)
    Name="LOCTEXT(\"Avg. PTS\")"
    Value=GetAverageMatchScore!(self:Fort Event Leaderboard Entry Data Object Reference=Get LeaderboardEntryData()) -> (ReturnValue:float)
    NumFractionalDigits=1
   )
   n3 = Branch(execute=@n1.then_4, Condition=Get bHasAltLeaderboardTrackedStats() -> (bHasAltLeaderboardTrackedStats:bool))
   > Set Alternative Stat Widget!(
    execute
    self
    Stat Widget:Event Leaderboard Details Stat Object Reference=Get Stat_EliminationsOrAltStat() -> (Stat_EliminationsOrAltStat:Event Leaderboard Details Stat Object Reference)
    Stats Names:Array of Strings=Get AltLeaderboardTrackedStatName() -> (AltLeaderboardTrackedStatName:Array of Strings)
    Stat Index:int=0
   ) -> (then) {is=title K2Node_CallFunction}
   > Set Alternative Stat Widget(
    Stat Widget=Get Stat_AltStat2() -> (Stat_AltStat2:Event Leaderboard Details Stat Object Reference)
    Stats Names=Get AltLeaderboardTrackedStatName()
    Stat Index=1
   )
   > Set Alternative Stat Widget(
    Stat Widget=Get Stat_AltStat3() -> (Stat_AltStat3:Event Leaderboard Details Stat Object Reference)
    Stats Names=Get AltLeaderboardTrackedStatName()
    Stat Index=2
   )
   > Set Alternative Stat Widget(
    Stat Widget=Get Stat_AltStat4() -> (Stat_AltStat4:Event Leaderboard Details Stat Object Reference)
    Stats Names=Get AltLeaderboardTrackedStatName()
    Stat Index=3
   )
   n4 = GetTotalEliminations!(self:Fort Event Leaderboard Entry Data Object Reference=Get LeaderboardEntryData()) -> (ReturnValue:int=0)
   n5 = GetTotalMatchesPlayed(self=Get LeaderboardEntryData()) -> (ReturnValue=0)
   SetStatData(
    execute=@n3.else
    self=Get Stat_EliminationsOrAltStat()
    Name="LOCTEXT(\"Avg. Elims\")"
    Value=float / float(A="To Float (Integer)"(InInt=@n4), B="To Float (Integer)"(InInt=@n5))
    NumFractionalDigits=1
   )
   > Set Visibility(Visibility=Collapsed, self=Get Stat_AltStat2())
   > Set Visibility(Visibility=Collapsed, self=Get Stat_AltStat3())
   > Set Visibility(Visibility=Collapsed, self=Get Stat_AltStat4())
   n6 = Get Stat_AveragePlacement() -> (Stat_AveragePlacement:Event Leaderboard Details Stat Object Reference)
   SetStatData(
    execute=@n1.then_2
    self=@n6
    Name="LOCTEXT(\"Avg. Placement\")"
    Value=GetAveragePlacement!(self:Fort Event Leaderboard Entry Data Object Reference=Get LeaderboardEntryData()) -> (ReturnValue:float)
    NumFractionalDigits=1
   )
   n7 = OR Boolean(A="Equal (Enum)"(A=Get EventType() -> (EventType=Default), B=Creative), B=Get bHasAltLeaderboardTrackedStats())
   > SetVisibility(self=@n6, InVisibility=Select(Option 0=SelfHitTestInvisible, Option 1=Collapsed, Index=@n7))
   n8 = Get LeaderboardEntryData()
   SetStatData(
    execute=@n1.then_3
    self=Get Stat_VictoryRoyales() -> (Stat_VictoryRoyales:Event Leaderboard Details Stat Object Reference)
    Name="LOCTEXT(\"Victory Royales\")"
    Value="To Float (Integer)"(InInt=GetTotalVictoryRoyales!(self:Fort Event Leaderboard Entry Data Object Reference=@n8) -> (ReturnValue:int=0))
    NumFractionalDigits=0
   )
   > SetVisibility(self=Get Stat_VictoryRoyales(), InVisibility=Select(Option 0=SelfHitTestInvisible, Option 1=Collapsed, Index=@n7))
  ]}
 SetTournamentValues {nodes [
   n1 = SetTournamentValues!() -> (then, InTournamentId:String, InEventWindowId:String) {is=title K2Node_FunctionEntry}
   > Set TournamentId(TournamentId:String=@n1) -> (Output_Get:String)
   > Set EventWindowId(EventWindowId:String=@n1.InEventWindowId) -> (Output_Get:String)
  ]}
 Show Details {nodes [
   n1 = Show Details!() -> (then, EntryData:Fort Event Leaderboard Entry Data Object Reference, EventType:ETournamentDisplayType Enum=Default) {is=title K2Node_FunctionEntry}
   > n2 = Set LeaderboardEntryData(LeaderboardEntryData:Fort Event Leaderboard Entry Data Object Reference=@n1) -> (Output_Get:Fort Event Leaderboard Entry Data Object Reference)
   > n3 = Is Valid(InputObject=@n2)
   > Update Alt Leaderboard Tracked Stats()
   n4 = Get LeaderboardEntryData()
   n5 = AND Boolean(A="integer > integer"(A=GetScore(self=@n4) -> (ReturnValue=0)), B=NOT Boolean(A=Get bHasAltLeaderboardTrackedStats()))
   > SetVisibility(
    self=Get Overlay_PointBreakdown() -> (Overlay_PointBreakdown:Overlay Object Reference)
    InVisibility=Select(Option 0=Hidden, Option 1=HitTestInvisible, Index=@n5)
   )
   > n6 = Branch(Condition=@n5)
   > Set Points Breakdown!(execute, self, EventType:ETournamentDisplayType Enum=Get EventType() -> (EventType=Default)) -> (then) {is=title K2Node_CallFunction}
   > Merge Execution!(True:Exec, False:Exec=@n6.else) -> (Then:Exec, FromTrueInput?:bool=~) {is=title K2Node_MacroInstance}
   > Set Stats Rows!(execute, self, EventType:ETournamentDisplayType Enum=Default) -> (then) {is=title K2Node_CallFunction}
   n7 = GetNumTeammates(self=Get LeaderboardEntryData()) -> (ReturnValue=0)
   > "SetText (Text)"(
    self=Get CommonTextBlock_Title()
    InText=Select!(
     Option 0:Text="LOCTEXT(\"STATS\")"
     Option 1:Text="LOCTEXT(\"PLAYER STATS\")"
     Option 2:Text="LOCTEXT(\"DUO STATS\")"
     Option 3:Text="LOCTEXT(\"TRIO STATS\")"
     Option 4:Text="LOCTEXT(\"SQUAD STATS\")"
     Index:int=@n7
    ) -> (ReturnValue:Text)
   )
   > "SetText (Text)"(
    self=Get CommonTextBlock_PlayerNames()
    InText=GetTeamDisplayNameText!(self:Fort Event Leaderboard Entry Data Object Reference=Get LeaderboardEntryData()) -> (ReturnValue:Text)
   )
   > "SetText (Text)"(
    self=Get CommonTextBlock_ProfileText() -> (CommonTextBlock_ProfileText:Common Text Object Reference)
    InText=Format Text!(Format="LOCTEXT(\"{0}|plural(one=Profile,other=Profiles)\")", 0:int=@n7) -> (Result)
   )
   > SetVisibility(
    self=Get Overlay_TeamMembers() -> (Overlay_TeamMembers:Overlay Object Reference)
    InVisibility=Select!(
     Option 0:ESlateVisibility Enum=SelfHitTestInvisible
     Option 1:ESlateVisibility Enum=SelfHitTestInvisible
     Index:bool="integer > integer"(A=@n7, B=1) {is=title K2Node_CallFunction}
    ) -> (ReturnValue:ESlateVisibility Enum)
   )
   > SetVisibility(InVisibility=SelfHitTestInvisible)
   > EventIntro() {is=title K2Node_CallFunction}
   n8 = GetLiveSessionId!(self:Fort Event Leaderboard Entry Data Object Reference=Get LeaderboardEntryData()) -> (ReturnValue:String)
   > SetVisibility(
    self=Get Image_LiveMatch() -> (Image_LiveMatch:Image Object Reference)
    InVisibility=Select(Option 0=SelfHitTestInvisible, Option 1=Collapsed, Index="Equal Exactly (String)"(A=@n8))
   )
   > SetVisibility(self=Get CommonBorderMainContent() -> (CommonBorderMainContent:Common Border Object Reference), InVisibility=SelfHitTestInvisible)
   SetVisibility(execute=@n3.Is Not Valid, self=Get CommonBorderMainContent(), InVisibility=Hidden)
   > SetVisibility(InVisibility=SelfHitTestInvisible)
  ]}
 Update Alt Leaderboard Tracked Stats {nodes [
   Update Alt Leaderboard Tracked Stats() {is=title K2Node_FunctionEntry}
   > n1 = Set AltLeaderboardTrackedStatName(AltLeaderboardTrackedStatName:Array of Strings=GetAlternativeLeaderboardTrackedStats!(
     self:Fort Leaderboard Entry Details Object Reference
     TournamentId:String=Get TournamentId()
     EventWindowId:String=Get EventWindowId()
     LeaderboardGroupingKey:String=Get LeaderboardGroupingKey()
     LeaderboardInstanceId:String=Get LeaderboardInstanceId()
    ) -> (ReturnValue:Array of Strings)) -> (Output_Get:Array of Strings)
   > Remove Item!(execute, TargetArray=@n1, Item:"String (by ref)") -> (then, ReturnValue:bool)
   n2 = Get AltLeaderboardTrackedStatName()
   > Set bHasAltLeaderboardTrackedStats(bHasAltLeaderboardTrackedStats:bool="integer > integer"(A=Length(TargetArray=@n2)) {is=title K2Node_PromotableOperator}) -> (Output_Get:bool)
  ]}
]
variables [
 AltLeaderboardTrackedStatName {type="array<string>", default []}
 bHasAltLeaderboardTrackedStats {type=boolean}
 CurrentDisplayInfo {
  type=FortTournamentDisplayInfo
  default {
   alertText=""
   alertType=Warning
   backgroundTitle=""
   detailsDescription=""
   flavorDescription=""
   longFormatTitle=""
   pinEarnedText=""
   pinScoreRequirement=0
   scheduleInfo=""
   seriesPointLeaderboardName=""
   shortFormatTitle=""
   titleLine1=""
   titleLine2=""
   backgroundLeftColor=<rgba 0 0 0 0>
   backgroundRightColor=<rgba 0 0 0 0>
   backgroundTextColor=<rgba 0 0 0 0>
   baseColor=<rgba 0 0 0 0>
   highlightColor=<rgba 0 0 0 0>
   posterFadeColor=<rgba 0 0 0 0>
   primaryColor=<rgba 0 0 0 0>
   roundDisplayNames []
   secondaryColor=<rgba 0 0 0 0>
   shadowColor=<rgba 0 0 0 0>
   titleColor=<rgba 0 0 0 0>
  }
 }
 EventWindowId {type=string, default=""}
 LeaderboardEntryData {type=FortEventLeaderboardEntryData, default=null}
 LeaderboardGroupingKey {type=string, default=""}
 LeaderboardInstanceId {type=string, default=""}
 ProfileSelectedFromSelector {}
 ProfileSelectorInstance {type=EventPlayerStatProfileSelector_C, default=null}
 ScoredPointsForPie {type=integer}
 TournamentId {type=string, default=""}
]
widgets {
 info {
  inheritedWidgetCount=0
  namedSlotCount=2
  parentClass=/Script/FortniteUI.FortLeaderboardEntryDetails
  rootWidgetClass=&SizeBox
  widgetCount=71
 }
 widgets [
  SizeBox_0 {widgetClassPath=&SizeBox, widget=SizeBox_0, properties {widthOverride=699.9756}}
  Overlay_0 {
   widgetClassPath=&Overlay
   parent=SizeBox_0
   widget=Overlay_0
   slot {class=SizeBoxSlot, properties {horizontalAlignment=HAlign_Right}}
  }
  Image_DetailsBG {
   widgetClassPath=&Image
   bIsVariable=true
   parent=Overlay_0
   widget=Image_DetailsBG
   properties {
    visibility=SelfHitTestInvisible
    brush {
     bIsDynamicallyLoaded=false
     drawAs=Box
     imageType=NoImage
     mirroring=NoMirror
     resourceName=None
     resourceObject=/CompeteOldTab/Frontend/Showdown/Materials/M_UI_SkewBox_LeaderboardDetailsBG
     tiling=NoTile
     imageSize=<967 1080>
     margin=<ltrb 0.4 0.2 0.4 0.2>
     outlineSettings {
      bUseBrushTransparency=false
      roundingType=HalfHeightRadius
      width=0
      color {colorUseRule=UseColor_Specified, specifiedColor=<rgba 0 0 0 0>}
      cornerRadii=<0 0 0 0>
     }
     tintColor {colorUseRule=UseColor_Specified, specifiedColor=<rgba 1 1 1 1>}
     uVRegion {bIsValid=false, max=<0 0>, min=<0 0>}
    }
    colorAndOpacity=<rgba 0.0009 0.0497 0.3185 1>
   }
   slot {class=OverlaySlot, properties {horizontalAlignment=HAlign_Right, verticalAlignment=VAlign_Fill, padding=<ltrb -400 0 0 0>}}
  }
  CommonBorderMainContent {
   widgetClassPath=&CommonBorder
   bIsVariable=true
   parent=Overlay_0
   widget=CommonBorderMainContent
   properties {
    background {
     bIsDynamicallyLoaded=false
     drawAs=NoDrawType
     imageType=NoImage
     mirroring=NoMirror
     resourceName=None
     resourceObject=null
     tiling=NoTile
     imageSize=<0 32>
     margin=<ltrb 0 0 0 0>
     outlineSettings {
      bUseBrushTransparency=false
      roundingType=HalfHeightRadius
      width=0
      color {colorUseRule=UseColor_Specified, specifiedColor=<rgba 0 0 0 0>}
      cornerRadii=<0 0 0 0>
     }
     tintColor {colorUseRule=UseColor_Specified, specifiedColor=<rgba 1 1 1 1>}
     uVRegion {bIsValid=false, max=<0 0>, min=<0 0>}
    }
    padding=<ltrb 0 0 0 0>
   }
   slot {class=OverlaySlot, properties {horizontalAlignment=HAlign_Center, verticalAlignment=VAlign_Fill}}
  }
  VerticalBox_1 {
   widgetClassPath=&VerticalBox
   parent=CommonBorderMainContent
   widget=VerticalBox_1
   properties {renderTransform {angle=0, scale=<1 1>, shear=<0 0>, translation=<-150 0>}}
   slot {class=BorderSlot, properties {padding=<ltrb 0 0 0 0>}}
  }
  CommonBorder_Title {
   widgetClassPath=&CommonBorder
   parent=VerticalBox_1
   widget=CommonBorder_Title
   properties {padding=<ltrb 0 0 0 0>}
   slot {class=VerticalBoxSlot, properties {horizontalAlignment=HAlign_Left, verticalAlignment=VAlign_Top, padding=<ltrb -50 100 -100 10>}}
  }
  HorizontalBox_0 {
   widgetClassPath=&HorizontalBox
   parent=CommonBorder_Title
   widget=HorizontalBox_0
   properties {renderTransform {angle=0, scale=<1 1>, shear=<0 0>, translation=<-31 0>}}
   slot {class=BorderSlot, properties {padding=<ltrb 0 0 0 0>}}
  }
  Image_32 {
   widgetClassPath=&Image
   bIsVariable=true
   parent=HorizontalBox_0
   widget=Image_32
   properties {
    visibility=Collapsed
    brush {
     bIsDynamicallyLoaded=false
     drawAs=Image
     imageType=NoImage
     mirroring=NoMirror
     resourceName=None
     resourceObject=/Game/UI/Foundation/Textures/Buttons/T_CycleArrow
     tiling=NoTile
     imageSize=<64 64>
     margin=<ltrb 0 0 0 0>
     outlineSettings {
      bUseBrushTransparency=false
      roundingType=HalfHeightRadius
      width=0
      color {colorUseRule=UseColor_Specified, specifiedColor=<rgba 0 0 0 0>}
      cornerRadii=<0 0 0 0>
     }
     tintColor {colorUseRule=UseColor_Specified, specifiedColor=<rgba 1 1 1 1>}
     uVRegion {bIsValid=false, max=<0 0>, min=<0 0>}
    }
    renderTransform {angle=0, scale=<-1 1>, shear=<0 0>, translation=<0 0>}
   }
   slot {class=HorizontalBoxSlot, properties {horizontalAlignment=HAlign_Center, verticalAlignment=VAlign_Center}}
  }
  SizeBox_1 {
   widgetClassPath=&SizeBox
   parent=HorizontalBox_0
   widget=SizeBox_1
   properties {bOverride_WidthOverride=true, maxDesiredWidth=690, widthOverride=690}
   slot {class=HorizontalBoxSlot, properties {padding=<ltrb 0 0 0 -10>}}
  }
  ScaleBox_0 {widgetClassPath=&ScaleBox, parent=SizeBox_1, widget=ScaleBox_0, slot {class=SizeBoxSlot}}
  CommonTextBlock_Title {
   widgetClassPath=&CommonTextBlock
   bIsVariable=true
   parent=ScaleBox_0
   widget=CommonTextBlock_Title
   properties {
    lineHeightPercentage=0.9
    text=TEAM STATS
    font {
     bForceMonospaced=false
     bMaterialIsStencil=false
     fontMaterial=null
     fontObject=&Burbank
     letterSpacing=0
     monospacedWidth=1
     size=65
     skewAmount=0
     typefaceFontName=BurbankRegular-Black
     outlineSettings {
      bApplyOutlineToDropShadows=false
      bMiteredCorners=false
      bSeparateFillAlpha=false
      outlineMaterial=null
      outlineSize=0
      outlineColor=<rgba 0 0 0 1>
     }
    }
    renderTransform {angle=0, scale=<1 1>, shear=<-5 0>, translation=<0 0>}
    shadowColorAndOpacity=<rgba 0.0009 0.0497 0.3185 1>
    shadowOffset=<8 8>
   }
   slot {class=ScaleBoxSlot}
  }
  Image_31 {
   widgetClassPath=&Image
   bIsVariable=true
   parent=HorizontalBox_0
   widget=Image_31
   properties {
    visibility=Collapsed
    brush {
     bIsDynamicallyLoaded=false
     drawAs=Image
     imageType=NoImage
     mirroring=NoMirror
     resourceName=None
     resourceObject=&T_CycleArrow
     tiling=NoTile
     imageSize=<64 64>
     margin=<ltrb 0 0 0 0>
     outlineSettings {
      bUseBrushTransparency=false
      roundingType=HalfHeightRadius
      width=0
      color {colorUseRule=UseColor_Specified, specifiedColor=<rgba 0 0 0 0>}
      cornerRadii=<0 0 0 0>
     }
     tintColor {colorUseRule=UseColor_Specified, specifiedColor=<rgba 1 1 1 1>}
     uVRegion {bIsValid=false, max=<0 0>, min=<0 0>}
    }
   }
   slot {class=HorizontalBoxSlot, properties {horizontalAlignment=HAlign_Center, verticalAlignment=VAlign_Center}}
  }
  VerticalBox_StatContainer {
   widgetClassPath=&VerticalBox
   bIsVariable=true
   parent=VerticalBox_1
   widget=VerticalBox_StatContainer
   properties {renderTransform {angle=0, scale=<1 1>, shear=<-12 0>, translation=<0 0>}}
   slot {class=VerticalBoxSlot, properties {padding=<ltrb 50 80 10 0>, size {sizeRule=Fill, value=1}}}
  }
  Stat_MatchesPlayed {
   widgetClassPath=/Game/UI/Events/Leaderboards/EventLeaderboardDetails_Stat
   bIsVariable=true
   parent=VerticalBox_StatContainer
   widget=Stat_MatchesPlayed
   properties {renderTransform {angle=0, scale=<1 1>, shear=<12 0>, translation=<0 0>}}
   slot {class=VerticalBoxSlot}
  }
  Stat_VictoryRoyales {
   widgetClassPath=&EventLeaderboardDetails_Stat
   bIsVariable=true
   parent=VerticalBox_StatContainer
   widget=Stat_VictoryRoyales
   properties {renderTransform {angle=0, scale=<1 1>, shear=<12 0>, translation=<0 0>}}
   slot {class=VerticalBoxSlot}
  }
  Stat_EliminationsOrAltStat {
   widgetClassPath=&EventLeaderboardDetails_Stat
   bIsVariable=true
   parent=VerticalBox_StatContainer
   widget=Stat_EliminationsOrAltStat
   properties {renderTransform {angle=0, scale=<1 1>, shear=<12 0>, translation=<0 0>}}
   slot {class=VerticalBoxSlot}
  }
  Stat_AltStat2 {
   widgetClassPath=&EventLeaderboardDetails_Stat
   bIsVariable=true
   parent=VerticalBox_StatContainer
   widget=Stat_AltStat2
   properties {renderTransform {angle=0, scale=<1 1>, shear=<12 0>, translation=<0 0>}}
   slot {class=VerticalBoxSlot}
  }
  Stat_AltStat3 {
   widgetClassPath=&EventLeaderboardDetails_Stat
   bIsVariable=true
   parent=VerticalBox_StatContainer
   widget=Stat_AltStat3
   properties {renderTransform {angle=0, scale=<1 1>, shear=<12 0>, translation=<0 0>}}
   slot {class=VerticalBoxSlot}
  }
  Stat_AltStat4 {
   widgetClassPath=&EventLeaderboardDetails_Stat
   bIsVariable=true
   parent=VerticalBox_StatContainer
   widget=Stat_AltStat4
   properties {renderTransform {angle=0, scale=<1 1>, shear=<12 0>, translation=<0 0>}}
   slot {class=VerticalBoxSlot}
  }
  Stat_AveragePlacement {
   widgetClassPath=&EventLeaderboardDetails_Stat
   bIsVariable=true
   parent=VerticalBox_StatContainer
   widget=Stat_AveragePlacement
   properties {renderTransform {angle=0, scale=<1 1>, shear=<12 0>, translation=<0 0>}}
   slot {class=VerticalBoxSlot}
  }
  Stat_AveragePointsPerMatch {
   widgetClassPath=&EventLeaderboardDetails_Stat
   bIsVariable=true
   parent=VerticalBox_StatContainer
   widget=Stat_AveragePointsPerMatch
   properties {renderTransform {angle=0, scale=<1 1>, shear=<12 0>, translation=<0 0>}}
   slot {class=VerticalBoxSlot}
  }
  Overlay_PointBreakdown {
   widgetClassPath=&Overlay
   bIsVariable=true
   parent=VerticalBox_1
   widget=Overlay_PointBreakdown
   properties {renderTransform {angle=0, scale=<1 1>, shear=<0 0>, translation=<-26 0>}}
   slot {class=VerticalBoxSlot, properties {verticalAlignment=VAlign_Top, padding=<ltrb 0 20 0 0>, size {sizeRule=Fill, value=1}}}
  }
  Overlay_3 {
   widgetClassPath=&Overlay
   parent=Overlay_PointBreakdown
   widget=Overlay_3
   slot {class=OverlaySlot, properties {horizontalAlignment=HAlign_Center, verticalAlignment=VAlign_Center}}
  }
  CommonBorder_1 {
   widgetClassPath=&CommonBorder
   parent=Overlay_3
   widget=CommonBorder_1
   properties {padding=<ltrb 0 0 0 0>, renderTransform {angle=0, scale=<1 1>, shear=<0 0>, translation=<10 15>}}
   slot {class=OverlaySlot}
  }
  Image_PieShadow {
   widgetClassPath=&Image
   bIsVariable=true
   parent=CommonBorder_1
   widget=Image_PieShadow
   properties {
    visibility=SelfHitTestInvisible
    brush {
     bIsDynamicallyLoaded=false
     drawAs=Image
     imageType=NoImage
     mirroring=NoMirror
     resourceName=None
     resourceObject=/Game/UI/Events/Leaderboards/Materials/M_UI_LeaderboardPieStat_Inst_WhiteFill
     tiling=NoTile
     imageSize=<255 255>
     margin=<ltrb 0 0 0 0>
     outlineSettings {
      bUseBrushTransparency=false
      roundingType=HalfHeightRadius
      width=0
      color {colorUseRule=UseColor_Specified, specifiedColor=<rgba 0 0 0 0>}
      cornerRadii=<0 0 0 0>
     }
     tintColor {colorUseRule=UseColor_Specified, specifiedColor=<rgba 1 1 1 1>}
     uVRegion {bIsValid=false, max=<0 0>, min=<0 0>}
    }
    colorAndOpacity=<rgba 0.0006 0.0212 0.1329 0.28>
   }
   slot {class=BorderSlot, properties {padding=<ltrb 0 0 0 0>}}
  }
  Image_Pie {
   widgetClassPath=&Image
   bIsVariable=true
   parent=Overlay_3
   widget=Image_Pie
   properties {
    visibility=SelfHitTestInvisible
    brush {
     bIsDynamicallyLoaded=false
     drawAs=Image
     imageType=NoImage
     mirroring=NoMirror
     resourceName=None
     resourceObject=/Game/UI/Events/Leaderboards/Materials/M_UI_LeaderboardPieStat_Inst
     tiling=NoTile
     imageSize=<255 255>
     margin=<ltrb 0 0 0 0>
     outlineSettings {
      bUseBrushTransparency=false
      roundingType=HalfHeightRadius
      width=0
      color {colorUseRule=UseColor_Specified, specifiedColor=<rgba 0 0 0 0>}
      cornerRadii=<0 0 0 0>
     }
     tintColor {colorUseRule=UseColor_Specified, specifiedColor=<rgba 1 1 1 1>}
     uVRegion {bIsValid=false, max=<0 0>, min=<0 0>}
    }
   }
   slot {class=OverlaySlot}
  }
  Overlay_4 {
   widgetClassPath=&Overlay
   parent=Overlay_PointBreakdown
   widget=Overlay_4
   slot {class=OverlaySlot, properties {horizontalAlignment=HAlign_Center, verticalAlignment=VAlign_Center, padding=<ltrb 0 0 -7 -5>}}
  }
  VerticalBox_3 {
   widgetClassPath=&VerticalBox
   parent=Overlay_4
   widget=VerticalBox_3
   slot {class=OverlaySlot, properties {verticalAlignment=VAlign_Center}}
  }
  SizeBox_4 {
   widgetClassPath=&SizeBox
   parent=VerticalBox_3
   widget=SizeBox_4
   properties {bOverride_MaxDesiredHeight=true, bOverride_MaxDesiredWidth=true, maxDesiredHeight=115, maxDesiredWidth=160}
   slot {class=VerticalBoxSlot}
  }
  ScaleBox_1 {widgetClassPath=&ScaleBox, parent=SizeBox_4, widget=ScaleBox_1, properties {stretch=ScaleToFitX}, slot {class=SizeBoxSlot}}
  CommonTextBlock_Score {
   widgetClassPath=&CommonTextBlock
   bIsVariable=true
   parent=ScaleBox_1
   widget=CommonTextBlock_Score
   properties {
    justification=Center
    text=40
    font {
     bForceMonospaced=false
     bMaterialIsStencil=false
     fontMaterial=null
     fontObject=&Burbank
     letterSpacing=0
     monospacedWidth=1
     size=90
     skewAmount=0
     typefaceFontName=Burbank-Black
     outlineSettings {
      bApplyOutlineToDropShadows=false
      bMiteredCorners=false
      bSeparateFillAlpha=false
      outlineMaterial=null
      outlineSize=0
      outlineColor=<rgba 0 0 0 1>
     }
    }
    margin=<ltrb 0 0 0 -10>
    shadowColorAndOpacity=<rgba 0.0006 0.0212 0.1329 0.28>
    shadowOffset=<10 15>
   }
   slot {class=ScaleBoxSlot}
  }
  CommonTextBlock_1 {
   widgetClassPath=&CommonTextBlock
   bIsVariable=true
   parent=VerticalBox_3
   widget=CommonTextBlock_1
   properties {
    justification=Center
    text=PTS
    font {
     bForceMonospaced=false
     bMaterialIsStencil=false
     fontMaterial=null
     fontObject=&Burbank
     letterSpacing=0
     monospacedWidth=1
     size=18
     skewAmount=0
     typefaceFontName=Burbank-PlayerPower
     outlineSettings {
      bApplyOutlineToDropShadows=false
      bMiteredCorners=false
      bSeparateFillAlpha=false
      outlineMaterial=null
      outlineSize=0
      outlineColor=<rgba 0 0 0 1>
     }
    }
    shadowColorAndOpacity=<rgba 0.0006 0.0212 0.1329 0.28>
    shadowOffset=<5 5>
   }
   slot {class=VerticalBoxSlot, properties {padding=<ltrb 0 -3 0 5>}}
  }
  VerticalBox_2 {
   widgetClassPath=&VerticalBox
   parent=Overlay_PointBreakdown
   widget=VerticalBox_2
   properties {renderTransform {angle=0, scale=<1 1>, shear=<0 0>, translation=<-32 0>}}
   slot {class=OverlaySlot}
  }
  Overlay_1 {
   widgetClassPath=&Overlay
   parent=VerticalBox_2
   widget=Overlay_1
   slot {class=VerticalBoxSlot, properties {horizontalAlignment=HAlign_Center, verticalAlignment=VAlign_Top}}
  }
  CommonBorder_ElimPointer {
   widgetClassPath=&CommonBorder
   bIsVariable=true
   parent=Overlay_1
   widget=CommonBorder_ElimPointer
   properties {padding=<ltrb 0 0 0 0>, renderTransform {angle=0, scale=<1 1>, shear=<0 0>, translation=<151 0>}}
   slot {class=OverlaySlot, properties {verticalAlignment=VAlign_Center, padding=<ltrb -100 0 -100 0>}}
  }
  HorizontalBox_1 {
   widgetClassPath=&HorizontalBox
   parent=CommonBorder_ElimPointer
   widget=HorizontalBox_1
   slot {class=BorderSlot, properties {padding=<ltrb 0 0 0 0>}}
  }
  CommonBorder_LeftStat {
   widgetClassPath=&CommonBorder
   parent=HorizontalBox_1
   widget=CommonBorder_LeftStat
   properties {padding=<ltrb 0 0 0 0>}
   slot {class=HorizontalBoxSlot, properties {verticalAlignment=VAlign_Center}}
  }
  SizeBox_2 {
   widgetClassPath=&SizeBox
   parent=CommonBorder_LeftStat
   widget=SizeBox_2
   properties {bOverride_HeightOverride=true, bOverride_WidthOverride=true, heightOverride=10, widthOverride=98.2043}
   slot {class=BorderSlot, properties {padding=<ltrb 0 0 0 0>}}
  }
  Image_57 {
   widgetClassPath=&Image
   bIsVariable=true
   parent=SizeBox_2
   widget=Image_57
   properties {brush {
     bIsDynamicallyLoaded=false
     drawAs=Box
     imageType=NoImage
     mirroring=NoMirror
     resourceName=None
     resourceObject=/Game/UI/Events/Leaderboards/Materials/M_UI_Leaderboard_StatPointer_Inst
     tiling=NoTile
     imageSize=<38 64>
     margin=<ltrb 0.3486 0 0 0>
     outlineSettings {
      bUseBrushTransparency=false
      roundingType=HalfHeightRadius
      width=0
      color {colorUseRule=UseColor_Specified, specifiedColor=<rgba 0 0 0 0>}
      cornerRadii=<0 0 0 0>
     }
     tintColor {colorUseRule=UseColor_Specified, specifiedColor=<rgba 1 1 1 1>}
     uVRegion {bIsValid=false, max=<0 0>, min=<0 0>}
    }}
   slot {class=SizeBoxSlot}
  }
  Image_58 {
   widgetClassPath=&Image
   bIsVariable=true
   parent=HorizontalBox_1
   widget=Image_58
   properties {
    brush {
     bIsDynamicallyLoaded=false
     drawAs=Image
     imageType=NoImage
     mirroring=NoMirror
     resourceName=None
     resourceObject=/Game/UI/Events/Leaderboards/Materials/M_UI_Leaderboard_StatPointer_Tail_Inst
     tiling=NoTile
     imageSize=<13 28>
     margin=<ltrb 0.211 0 0.2 0>
     outlineSettings {
      bUseBrushTransparency=false
      roundingType=HalfHeightRadius
      width=0
      color {colorUseRule=UseColor_Specified, specifiedColor=<rgba 0 0 0 0>}
      cornerRadii=<0 0 0 0>
     }
     tintColor {colorUseRule=UseColor_Specified, specifiedColor=<rgba 1 1 1 1>}
     uVRegion {bIsValid=false, max=<0 0>, min=<0 0>}
    }
    renderTransformPivot=<0 0.5>
   }
   slot {class=HorizontalBoxSlot}
  }
  Image_ElimIcon {
   widgetClassPath=&Image
   bIsVariable=true
   parent=Overlay_1
   widget=Image_ElimIcon
   properties {
    brush {
     bIsDynamicallyLoaded=false
     drawAs=Image
     imageType=NoImage
     mirroring=NoMirror
     resourceName=None
     resourceObject=/Game/UI/Events/Leaderboards/Materials/M_UI_Leaderboard_StatIUcon_Eliminations
     tiling=NoTile
     imageSize=<44 44>
     margin=<ltrb 0 0 0 0>
     outlineSettings {
      bUseBrushTransparency=false
      roundingType=HalfHeightRadius
      width=0
      color {colorUseRule=UseColor_Specified, specifiedColor=<rgba 0 0 0 0>}
      cornerRadii=<0 0 0 0>
     }
     tintColor {colorUseRule=UseColor_Specified, specifiedColor=<rgba 1 1 1 1>}
     uVRegion {bIsValid=false, max=<0 0>, min=<0 0>}
    }
    colorAndOpacity=<rgba 0.0185 1 0.9047 1>
   }
   slot {class=OverlaySlot}
  }
  SizeBox_6 {
   widgetClassPath=&SizeBox
   parent=VerticalBox_2
   widget=SizeBox_6
   properties {bOverride_WidthOverride=true, maxDesiredWidth=250, widthOverride=250}
   slot {class=VerticalBoxSlot}
  }
  VerticalBox_0 {widgetClassPath=&VerticalBox, parent=SizeBox_6, widget=VerticalBox_0, slot {class=SizeBoxSlot}}
  CommonTextBlock_0 {
   widgetClassPath=&CommonTextBlock
   parent=VerticalBox_0
   widget=CommonTextBlock_0
   properties {
    justification=Center
    text=ELIMINATIONS
    font {
     bForceMonospaced=false
     bMaterialIsStencil=false
     fontMaterial=null
     fontObject=&Burbank
     letterSpacing=0
     monospacedWidth=1
     size=25
     skewAmount=0
     typefaceFontName=Burbank-PlayerPower
     outlineSettings {
      bApplyOutlineToDropShadows=false
      bMiteredCorners=false
      bSeparateFillAlpha=false
      outlineMaterial=null
      outlineSize=0
      outlineColor=<rgba 0 0 0 1>
     }
    }
    shadowOffset=<0 0>
   }
   slot {class=VerticalBoxSlot}
  }
  Text_EliminationPointInfo {
   widgetClassPath=&CommonTextBlock
   bIsVariable=true
   parent=VerticalBox_0
   widget=Text_EliminationPointInfo
   properties {
    justification=Center
    text=(0 PTS - 0%)
    font {
     bForceMonospaced=false
     bMaterialIsStencil=false
     fontMaterial=null
     fontObject=&Burbank
     letterSpacing=0
     monospacedWidth=1
     size=18
     skewAmount=0
     typefaceFontName=Burbank-PlayerPower
     outlineSettings {
      bApplyOutlineToDropShadows=false
      bMiteredCorners=false
      bSeparateFillAlpha=false
      outlineMaterial=null
      outlineSize=0
      outlineColor=<rgba 0 0 0 1>
     }
    }
    shadowOffset=<0 0>
   }
   slot {class=VerticalBoxSlot}
  }
  VerticalBox_Placement {
   widgetClassPath=&VerticalBox
   bIsVariable=true
   parent=Overlay_PointBreakdown
   widget=VerticalBox_Placement
   properties {renderTransform {angle=0, scale=<1 1>, shear=<0 0>, translation=<40 10>}}
   slot {class=OverlaySlot, properties {horizontalAlignment=HAlign_Right, verticalAlignment=VAlign_Bottom, padding=<ltrb 0 0 35 0>}}
  }
  Overlay_Placement {
   widgetClassPath=&Overlay
   bIsVariable=true
   parent=VerticalBox_Placement
   widget=Overlay_Placement
   slot {class=VerticalBoxSlot, properties {horizontalAlignment=HAlign_Center, verticalAlignment=VAlign_Top}}
  }
  CommonBorder_PlacementPointer {
   widgetClassPath=&CommonBorder
   bIsVariable=true
   parent=Overlay_Placement
   widget=CommonBorder_PlacementPointer
   properties {padding=<ltrb 0 0 0 0>, renderTransform {angle=0, scale=<-1 -1>, shear=<0 0>, translation=<-219 0>}, renderTransformPivot=<1 0.5>}
   slot {class=OverlaySlot, properties {horizontalAlignment=HAlign_Right, verticalAlignment=VAlign_Center, padding=<ltrb -100 0 -100 0>}}
  }
  HorizontalBox_2 {
   widgetClassPath=&HorizontalBox
   parent=CommonBorder_PlacementPointer
   widget=HorizontalBox_2
   slot {class=BorderSlot, properties {padding=<ltrb 0 0 0 0>}}
  }
  CommonBorder_RightStat {
   widgetClassPath=&CommonBorder
   parent=HorizontalBox_2
   widget=CommonBorder_RightStat
   properties {horizontalAlignment=HAlign_Right, padding=<ltrb 0 0 0 0>}
   slot {class=HorizontalBoxSlot, properties {horizontalAlignment=HAlign_Right, verticalAlignment=VAlign_Center}}
  }
  SizeBox_3 {
   widgetClassPath=&SizeBox
   parent=CommonBorder_RightStat
   widget=SizeBox_3
   properties {bOverride_HeightOverride=true, bOverride_WidthOverride=true, heightOverride=10, widthOverride=51.984}
   slot {class=BorderSlot, properties {horizontalAlignment=HAlign_Right, padding=<ltrb 0 0 0 0>}}
  }
  Image_93 {
   widgetClassPath=&Image
   bIsVariable=true
   parent=SizeBox_3
   widget=Image_93
   properties {brush {
     bIsDynamicallyLoaded=false
     drawAs=Box
     imageType=NoImage
     mirroring=NoMirror
     resourceName=None
     resourceObject=&M_UI_Leaderboard_StatPointer_Inst
     tiling=NoTile
     imageSize=<38 64>
     margin=<ltrb 0.3486 0 0 0>
     outlineSettings {
      bUseBrushTransparency=false
      roundingType=HalfHeightRadius
      width=0
      color {colorUseRule=UseColor_Specified, specifiedColor=<rgba 0 0 0 0>}
      cornerRadii=<0 0 0 0>
     }
     tintColor {colorUseRule=UseColor_Specified, specifiedColor=<rgba 1 1 1 1>}
     uVRegion {bIsValid=false, max=<0 0>, min=<0 0>}
    }}
   slot {class=SizeBoxSlot}
  }
  Image_94 {
   widgetClassPath=&Image
   bIsVariable=true
   parent=HorizontalBox_2
   widget=Image_94
   properties {
    brush {
     bIsDynamicallyLoaded=false
     drawAs=Image
     imageType=NoImage
     mirroring=NoMirror
     resourceName=None
     resourceObject=&M_UI_Leaderboard_StatPointer_Tail_Inst
     tiling=NoTile
     imageSize=<13 28>
     margin=<ltrb 0.211 0 0.2 0>
     outlineSettings {
      bUseBrushTransparency=false
      roundingType=HalfHeightRadius
      width=0
      color {colorUseRule=UseColor_Specified, specifiedColor=<rgba 0 0 0 0>}
      cornerRadii=<0 0 0 0>
     }
     tintColor {colorUseRule=UseColor_Specified, specifiedColor=<rgba 1 1 1 1>}
     uVRegion {bIsValid=false, max=<0 0>, min=<0 0>}
    }
    renderTransformPivot=<0 0.5>
   }
   slot {class=HorizontalBoxSlot}
  }
  Image_PlacementIcon {
   widgetClassPath=&Image
   bIsVariable=true
   parent=Overlay_Placement
   widget=Image_PlacementIcon
   properties {
    brush {
     bIsDynamicallyLoaded=false
     drawAs=Image
     imageType=NoImage
     mirroring=NoMirror
     resourceName=None
     resourceObject=/Game/UI/Events/Leaderboards/Materials/M_UI_Leaderboard_StatIUcon_Placement
     tiling=NoTile
     imageSize=<44 44>
     margin=<ltrb 0 0 0 0>
     outlineSettings {
      bUseBrushTransparency=false
      roundingType=HalfHeightRadius
      width=0
      color {colorUseRule=UseColor_Specified, specifiedColor=<rgba 0 0 0 0>}
      cornerRadii=<0 0 0 0>
     }
     tintColor {colorUseRule=UseColor_Specified, specifiedColor=<rgba 1 1 1 1>}
     uVRegion {bIsValid=false, max=<0 0>, min=<0 0>}
    }
    colorAndOpacity=<rgba 0.008 1 0.2384 1>
   }
   slot {class=OverlaySlot}
  }
  SizeBox_7 {
   widgetClassPath=&SizeBox
   parent=VerticalBox_Placement
   widget=SizeBox_7
   properties {bOverride_WidthOverride=true, widthOverride=250}
   slot {class=VerticalBoxSlot}
  }
  VerticalBox_4 {widgetClassPath=&VerticalBox, parent=SizeBox_7, widget=VerticalBox_4, slot {class=SizeBoxSlot}}
  CommonTextBlock_Placement {
   widgetClassPath=&CommonTextBlock
   bIsVariable=true
   parent=VerticalBox_4
   widget=CommonTextBlock_Placement
   properties {
    justification=Center
    text=PLACEMENT
    font {
     bForceMonospaced=false
     bMaterialIsStencil=false
     fontMaterial=null
     fontObject=&Burbank
     letterSpacing=0
     monospacedWidth=1
     size=25
     skewAmount=0
     typefaceFontName=Burbank-PlayerPower
     outlineSettings {
      bApplyOutlineToDropShadows=false
      bMiteredCorners=false
      bSeparateFillAlpha=false
      outlineMaterial=null
      outlineSize=0
      outlineColor=<rgba 0 0 0 1>
     }
    }
    shadowOffset=<0 0>
   }
   slot {class=VerticalBoxSlot}
  }
  Text_PlacementPointInfo {
   widgetClassPath=&CommonTextBlock
   bIsVariable=true
   parent=VerticalBox_4
   widget=Text_PlacementPointInfo
   properties {
    justification=Center
    text=(0 PTS - 0%)
    font {
     bForceMonospaced=false
     bMaterialIsStencil=false
     fontMaterial=null
     fontObject=&Burbank
     letterSpacing=0
     monospacedWidth=1
     size=18
     skewAmount=0
     typefaceFontName=Burbank-PlayerPower
     outlineSettings {
      bApplyOutlineToDropShadows=false
      bMiteredCorners=false
      bSeparateFillAlpha=false
      outlineMaterial=null
      outlineSize=0
      outlineColor=<rgba 0 0 0 1>
     }
    }
    shadowOffset=<0 0>
   }
   slot {class=VerticalBoxSlot}
  }
  Overlay_TeamMembers {
   widgetClassPath=&Overlay
   bIsVariable=true
   parent=VerticalBox_1
   widget=Overlay_TeamMembers
   properties {renderTransform {angle=0, scale=<1 1>, shear=<0 0>, translation=<-62 0>}}
   slot {class=VerticalBoxSlot, properties {verticalAlignment=VAlign_Bottom, padding=<ltrb -70 -35 -30 85>}}
  }
  Image_PlayernameBG_Shadow {
   widgetClassPath=&Image
   bIsVariable=true
   parent=Overlay_TeamMembers
   widget=Image_PlayernameBG_Shadow
   properties {
    visibility=SelfHitTestInvisible
    brush {
     bIsDynamicallyLoaded=false
     drawAs=Box
     imageType=NoImage
     mirroring=NoMirror
     resourceName=None
     resourceObject=/Game/UI/Frontend/Showdown/Materials/M_UI_SkewBox_LeaderboardDetailsBG2
     tiling=NoTile
     imageSize=<773 57>
     margin=<ltrb 0.2 0.2 0.2 0.2>
     outlineSettings {
      bUseBrushTransparency=false
      roundingType=HalfHeightRadius
      width=0
      color {colorUseRule=UseColor_Specified, specifiedColor=<rgba 0 0 0 0>}
      cornerRadii=<0 0 0 0>
     }
     tintColor {colorUseRule=UseColor_Specified, specifiedColor=<rgba 1 1 1 1>}
     uVRegion {bIsValid=false, max=<0 0>, min=<0 0>}
    }
    colorAndOpacity=<rgba 0.0006 0.0212 0.1329 0.28>
    renderTransform {angle=0, scale=<1 1>, shear=<0 0>, translation=<20 20>}
   }
   slot {class=OverlaySlot, properties {horizontalAlignment=HAlign_Fill, verticalAlignment=VAlign_Fill, padding=<ltrb -30 0 -5 25>}}
  }
  Image_PlayernameBG {
   widgetClassPath=&Image
   bIsVariable=true
   parent=Overlay_TeamMembers
   widget=Image_PlayernameBG
   properties {
    brush {
     bIsDynamicallyLoaded=false
     drawAs=Box
     imageType=NoImage
     mirroring=NoMirror
     resourceName=None
     resourceObject=&M_UI_SkewBox_LeaderboardDetailsBG2
     tiling=NoTile
     imageSize=<773 57>
     margin=<ltrb 0.2 0.2 0.2 0.2>
     outlineSettings {
      bUseBrushTransparency=false
      roundingType=HalfHeightRadius
      width=0
      color {colorUseRule=UseColor_Specified, specifiedColor=<rgba 0 0 0 0>}
      cornerRadii=<0 0 0 0>
     }
     tintColor {colorUseRule=UseColor_Specified, specifiedColor=<rgba 1 1 1 1>}
     uVRegion {bIsValid=false, max=<0 0>, min=<0 0>}
    }
    colorAndOpacity=<rgba 0.008 1 0.2384 1>
   }
   slot {class=OverlaySlot, properties {horizontalAlignment=HAlign_Fill, verticalAlignment=VAlign_Fill, padding=<ltrb -25 0 -10 20>}}
  }
  SizeBox_5 {
   widgetClassPath=&SizeBox
   parent=Overlay_TeamMembers
   widget=SizeBox_5
   properties {bOverride_WidthOverride=true, widthOverride=685.7302}
   slot {class=OverlaySlot, properties {horizontalAlignment=HAlign_Center, verticalAlignment=VAlign_Center, padding=<ltrb 35 5 45 10>}}
  }
  CommonTextBlock_PlayerNames {
   widgetClassPath=&CommonTextBlock
   bIsVariable=true
   parent=SizeBox_5
   widget=CommonTextBlock_PlayerNames
   properties {
    scrollStyle=&TextScrollStyle-Base_C
    text="LOREMIPSUM, DOLOR_AMIT, DAPIBUS_ET_CONGUE, TEMPUS HENDRERIT"
    colorAndOpacity {colorUseRule=UseColor_Specified, specifiedColor=<rgba 0.0009 0.0497 0.3185 1>}
    font {
     bForceMonospaced=false
     bMaterialIsStencil=false
     fontMaterial=null
     fontObject=&Burbank
     letterSpacing=0
     monospacedWidth=1
     size=22
     skewAmount=0
     typefaceFontName=Burbank-PlayerPower
     outlineSettings {
      bApplyOutlineToDropShadows=false
      bMiteredCorners=false
      bSeparateFillAlpha=false
      outlineMaterial=null
      outlineSize=0
      outlineColor=<rgba 0 0 0 1>
     }
    }
   }
   slot {class=SizeBoxSlot}
  }
  CommonTextBlock_PlayerNamesTitle {
   widgetClassPath=&CommonTextBlock
   bIsVariable=true
   parent=Overlay_TeamMembers
   widget=CommonTextBlock_PlayerNamesTitle
   properties {
    text=WITH
    colorAndOpacity {colorUseRule=UseColor_Specified, specifiedColor=<rgba 0.0185 1 0.9047 1>}
    font {
     bForceMonospaced=false
     bMaterialIsStencil=false
     fontMaterial=null
     fontObject=&Burbank
     letterSpacing=0
     monospacedWidth=1
     size=24
     skewAmount=0
     typefaceFontName=Burbank-PlayerPower
     outlineSettings {
      bApplyOutlineToDropShadows=false
      bMiteredCorners=false
      bSeparateFillAlpha=false
      outlineMaterial=null
      outlineSize=3
      outlineColor=<rgba 0.0176 0.0953 0.7157 1>
     }
    }
    renderTransform {angle=-7, scale=<1 1>, shear=<-8 0>, translation=<-11 -20>}
    shadowOffset=<0 0>
   }
   slot {class=OverlaySlot}
  }
  HorizontalBox_3 {
   widgetClassPath=&HorizontalBox
   parent=Overlay_TeamMembers
   widget=HorizontalBox_3
   slot {class=OverlaySlot, properties {horizontalAlignment=HAlign_Center, verticalAlignment=VAlign_Bottom, padding=<ltrb 0 0 75 -50>}}
  }
  Button_MatchHistory {
   widgetClassPath=&IconTextButton
   bIsVariable=true
   parent=HorizontalBox_3
   widget=Button_MatchHistory
   properties {
    bIsFocusable=false
    buttonText=""
    buttonTextJustification=Right
    onButtonBaseClicked=(&self:WidgetTree.Button_MatchHistory.HandleOnButtonClicked)
    onButtonBaseDoubleClicked=(&self:WidgetTree.Button_MatchHistory.HandleOnButtonDoubleClicked)
    onButtonBaseHovered=(&self:WidgetTree.Button_MatchHistory.HandleOnButtonHovered)
    onButtonBaseUnhovered=(&self:WidgetTree.Button_MatchHistory.HandleOnButtonUnhovered)
    onSelectedChangedBase=(&self:WidgetTree.Button_MatchHistory.HandleOnSelectedChanged)
    style=/FortUILibrary/Foundation/Buttons/ButtonStyles/ButtonStyle-TextOnlyBase_EmptyFocus.ButtonStyle-TextOnlyBase_EmptyFocus_C
    triggeringInputAction {dataTable=&InputActionDataTable, rowName=GenericForward}
   }
   slot {class=HorizontalBoxSlot, properties {horizontalAlignment=HAlign_Left, verticalAlignment=VAlign_Top}}
  }
  HorizontalBox_4 {widgetClassPath=&HorizontalBox, namedSlotHost=Button_MatchHistory, widget=HorizontalBox_4}
  Image_LiveMatch {
   widgetClassPath=&Image
   bIsVariable=true
   parent=HorizontalBox_4
   widget=Image_LiveMatch
   properties {
    brush {
     bIsDynamicallyLoaded=false
     drawAs=Image
     imageType=NoImage
     mirroring=NoMirror
     resourceName=None
     resourceObject=/Game/UI/Events/Leaderboards/Materials/M_UI_LiveIconCenter_Inst
     tiling=NoTile
     imageSize=<32 32>
     margin=<ltrb 0 0 0 0>
     outlineSettings {
      bUseBrushTransparency=false
      roundingType=HalfHeightRadius
      width=0
      color {colorUseRule=UseColor_Specified, specifiedColor=<rgba 0 0 0 0>}
      cornerRadii=<0 0 0 0>
     }
     tintColor {colorUseRule=UseColor_Specified, specifiedColor=<rgba 1 1 1 1>}
     uVRegion {bIsValid=false, max=<0 0>, min=<0 0>}
    }
    colorAndOpacity=<rgba 1 0 0 1>
   }
   slot {class=HorizontalBoxSlot, properties {horizontalAlignment=HAlign_Center, verticalAlignment=VAlign_Top, padding=<ltrb 0 5 -2 0>}}
  }
  CommonTextBlock_MatchHistoryTextMatchHistory {
   widgetClassPath=&CommonTextBlock
   bIsVariable=true
   parent=HorizontalBox_4
   widget=CommonTextBlock_MatchHistoryTextMatchHistory
   properties {
    text=Match History
    textTransformPolicy=ToUpper
    font {
     bForceMonospaced=false
     bMaterialIsStencil=false
     fontMaterial=null
     fontObject=&BurbankSmall
     letterSpacing=0
     monospacedWidth=1
     size=24
     skewAmount=0
     typefaceFontName=Bold
     outlineSettings {
      bApplyOutlineToDropShadows=false
      bMiteredCorners=false
      bSeparateFillAlpha=false
      outlineMaterial=null
      outlineSize=0
      outlineColor=<rgba 0 0 0 1>
     }
    }
    shadowOffset=<0 0>
   }
   slot {class=HorizontalBoxSlot}
  }
  Button_Profiles {
   widgetClassPath=&IconTextButton
   bIsVariable=true
   parent=HorizontalBox_3
   widget=Button_Profiles
   properties {
    bIsFocusable=false
    buttonText=""
    buttonTextJustification=Right
    onButtonBaseClicked=(&self:WidgetTree.Button_Profiles.HandleOnButtonClicked)
    onButtonBaseDoubleClicked=(&self:WidgetTree.Button_Profiles.HandleOnButtonDoubleClicked)
    onButtonBaseHovered=(&self:WidgetTree.Button_Profiles.HandleOnButtonHovered)
    onButtonBaseUnhovered=(&self:WidgetTree.Button_Profiles.HandleOnButtonUnhovered)
    onSelectedChangedBase=(&self:WidgetTree.Button_Profiles.HandleOnSelectedChanged)
    style=&ButtonStyle-TextOnlyBase_EmptyFocus_C
    triggeringInputAction {dataTable=&InputActionDataTable, rowName=EventLeaderboardRowInteraction}
   }
   slot {class=HorizontalBoxSlot, properties {padding=<ltrb 80 0 0 0>}}
  }
  CommonTextBlock_ProfileText {
   widgetClassPath=&CommonTextBlock
   bIsVariable=true
   namedSlotHost=Button_Profiles
   widget=CommonTextBlock_ProfileText
   properties {
    text=Profiles
    textTransformPolicy=ToUpper
    font {
     bForceMonospaced=false
     bMaterialIsStencil=false
     fontMaterial=null
     fontObject=&BurbankSmall
     letterSpacing=0
     monospacedWidth=1
     size=24
     skewAmount=0
     typefaceFontName=Bold
     outlineSettings {
      bApplyOutlineToDropShadows=false
      bMiteredCorners=false
      bSeparateFillAlpha=false
      outlineMaterial=null
      outlineSize=0
      outlineColor=<rgba 0 0 0 1>
     }
    }
    shadowOffset=<0 0>
   }
  }
 ]
}
