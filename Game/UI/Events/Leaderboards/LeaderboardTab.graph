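# Fortnite tournament leaderboard tab widget, derived from FortLeaderboardViewTab, that shows an event's leaderboard entries over a blurred, tinted panel.
# An inherited switcher holds the list wrapper and a "No Leaderboard Entries" message, alongside a hidden debug ID text.
# Colorize tints the container with the tournament's left background color at half opacity and passes the display info to the list wrapper.
# Set Event forwards event and window IDs to that wrapper.
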WidgetBlueprint /Game/UI/Events/Leaderboards/LeaderboardTab
graphs [EventGraph {nodes [
   n1 = Colorize!() -> (OutputDelegate, then, Color Info:Fort Tournament Display Info Structure) {is=title K2Node_CustomEvent}
   n2 = Break Linear Color!(LinearColor:"Linear Color Structure (by ref)"=Break Fort Tournament Display Info!(FortTournamentDisplayInfo:"Fort Tournament Display Info Structure (by ref)"=@n1.Color Info) -> (
     TitleLine1:Text
     TitleLine2:Text
     ScheduleInfo:Text
     FlavorDescription:Text
     DetailsDescription:Text
     ShortFormatTitle:Text
     LongFormatTitle:Text
     BackgroundTitle:Text
     PinScoreRequirement:int
     PinEarnedText:Text
     BaseColor:Linear Color Structure
     PrimaryColor:Linear Color Structure
     SecondaryColor:Linear Color Structure
     HighlightColor:Linear Color Structure
     TitleColor:Linear Color Structure
     ShadowColor:Linear Color Structure
     BackgroundLeftColor:Linear Color Structure
     BackgroundRightColor:Linear Color Structure
     BackgroundTextColor:Linear Color Structure
     PosterFadeColor:Linear Color Structure
     AlertText:Text
     SeriesPointLeaderboardName:Text
     AlertType:EFortTournamentAlertType Enum
     RoundDisplayNames:Array of Texts
    ) {is=title K2Node_BreakStruct}.BackgroundLeftColor) -> (R:float=~, G:float=~, B:float=~, A:float=~) {is=title K2Node_BreakStruct}
   > SetBrushColor(
    self=Get CommonBorder_ContainerBG() -> (CommonBorder_ContainerBG:Common Border Object Reference)
    InBrushColor=Make Linear Color!(R:float=@n2, G:float=@n2.G, B:float=@n2.B, A:float=0.5) -> (LinearColor:Linear Color Structure) {is=title K2Node_MakeStruct}
   )
   > n3 = Cast To LeaderboardListViewWrapper!(
    execute
    Object=Get ActivatableContent_LeaderboardListWrapper() -> (ActivatableContent_LeaderboardListWrapper:Fort Leaderboard List View Wrapper Object Reference)
   ) -> (then, CastFailed, AsLeaderboard List View Wrapper:Leaderboard List View Wrapper Object Reference) {is=title K2Node_DynamicCast}
   > Set Display Info!(execute, self:Leaderboard List View Wrapper Object Reference=@n3, DisplayInfo:Fort Tournament Display Info Structure=@n1.Color Info) -> (then) {is=title K2Node_CallFunction}
   n4 = Set Event!() -> (OutputDelegate, then, EventId:String, EventWindowId:String) {is=title K2Node_CustomEvent}
   > n5 = Cast To LeaderboardListViewWrapper(Object=Get ActivatableContent_LeaderboardListWrapper())
   > Set Event!(execute, self:Leaderboard List View Wrapper Object Reference=@n5, EventId:String=@n4.EventId, EventWindowId:String=@n4.EventWindowId) -> (then) {is=title K2Node_CallFunction}
  ]}]
variables []
widgets {
 info {inheritedWidgetCount=4, parentClass=/Script/FortniteUI.FortLeaderboardViewTab, rootWidgetClass=&Overlay, widgetCount=9}
 widgets [
  Overlay_LeaderboardContainer {widgetClassPath=&Overlay, widget=Overlay_LeaderboardContainer}
  BackgroundBlur_0 {
   widgetClassPath=&BackgroundBlur
   parent=Overlay_LeaderboardContainer
   widget=BackgroundBlur_0
   properties {blurStrength=50}
   slot {class=OverlaySlot, properties {horizontalAlignment=HAlign_Fill, verticalAlignment=VAlign_Fill, padding=<ltrb 50 0 20 0>}}
  }
  SizeBox_1 {
   widgetClassPath=&SizeBox
   parent=BackgroundBlur_0
   widget=SizeBox_1
   properties {bOverride_MinDesiredHeight=true, minDesiredHeight=180}
   slot {class=BackgroundBlurSlot, properties {padding=<ltrb 0 0 0 0>}}
  }
  CommonBorder_ContainerBG {
   widgetClassPath=&CommonBorder
   bIsVariable=true
   parent=SizeBox_1
   widget=CommonBorder_ContainerBG
   properties {brushColor=<rgba 0.0052 0.0723 0.4179 0.5>, padding=<ltrb 0 0 0 0>}
   slot {class=SizeBoxSlot}
  }
  Switcher_LeaderboardEntries {
   widgetClassPath=&CommonWidgetSwitcherLegacy
   bInherited=true
   bIsVariable=true
   parent=CommonBorder_ContainerBG
   widget=Switcher_LeaderboardEntries
   properties {transitionDuration=0}
   slot {class=BorderSlot, properties {padding=<ltrb 0 0 0 0>}}
  }
  ActivatableContent_LeaderboardListWrapper {
   widgetClassPath=&/LeaderboardListViewWrapper
   bInherited=true
   bIsVariable=true
   parent=Switcher_LeaderboardEntries
   widget=ActivatableContent_LeaderboardListWrapper
   slot {class=WidgetSwitcherSlot}
  }
  Content_NoLeaderboardEntries {
   widgetClassPath=&CommonBorder
   bInherited=true
   bIsVariable=true
   parent=Switcher_LeaderboardEntries
   widget=Content_NoLeaderboardEntries
   properties {verticalAlignment=VAlign_Top, padding=<ltrb 10 300 10 25>}
   slot {class=WidgetSwitcherSlot}
  }
  Text_NoLeaderboardEntries {
   widgetClassPath=&CommonTextBlock
   bIsVariable=true
   parent=Content_NoLeaderboardEntries
   widget=Text_NoLeaderboardEntries
   properties {
    autoWrapText=true
    justification=Center
    text="No Leaderboard Entries Available\r\nCheck Back Later"
    font {
     bForceMonospaced=false
     bMaterialIsStencil=false
     fontMaterial=null
     fontObject=&Burbank
     letterSpacing=0
     monospacedWidth=1
     size=24
     skewAmount=0
     typefaceFontName=Burbank-PlayerPower
     outlineSettings {
      bApplyOutlineToDropShadows=false
      bMiteredCorners=false
      bSeparateFillAlpha=false
      outlineMaterial=null
      outlineSize=0
      outlineColor=<rgba 0 0 0 1>
     }
    }
    shadowOffset=<0 0>
   }
   slot {class=BorderSlot, properties {verticalAlignment=VAlign_Top, padding=<ltrb 10 300 10 25>}}
  }
  Text_DebugId {
   widgetClassPath=&TextBlock
   bInherited=true
   parent=Overlay_LeaderboardContainer
   widget=Text_DebugId
   properties {
    autoWrapText=true
    justification=Center
    text=EventId: Event EventWindowId: EventWindow
    visibility=Hidden
    wrapTextAt=600
    font {
     bForceMonospaced=false
     bMaterialIsStencil=false
     fontMaterial=null
     fontObject=&Roboto
     letterSpacing=0
     monospacedWidth=1
     size=24
     skewAmount=0
     typefaceFontName=Bold
     outlineSettings {
      bApplyOutlineToDropShadows=false
      bMiteredCorners=false
      bSeparateFillAlpha=false
      outlineMaterial=null
      outlineSize=3
      outlineColor=<rgba 0 0 0 1>
     }
    }
   }
   slot {class=OverlaySlot, properties {horizontalAlignment=HAlign_Fill, padding=<ltrb 0 200 0 0>}}
  }
 ]
}
